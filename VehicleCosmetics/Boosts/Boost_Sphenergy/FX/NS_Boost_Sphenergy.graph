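# Niagara system for the Sphenergy vehicle boost cosmetic, drawing rings, smoke and trails offset from the boost position.
# Enabled emitters such as Rings001 and SMOKECORE002 merge several layers, using spawn groups to pick lifetime, size, velocity and scale for their mesh and sprite renderers.
# Painted-variant overrides set per-layer colors, while user parameters scale painted brightness and adjust smoke for frontend preview.
# Several single-layer emitters remain disabled.
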NiagaraSystem /VehicleCosmetics/Boosts/Boost_Sphenergy/FX/NS_Boost_Sphenergy
dependencies {
 usedDataInterfaces [
  &NiagaraDataInterfaceCurve
  &NiagaraDataInterfaceArrayColor
  &NiagaraDataInterfaceVectorCurve
  &NiagaraDataInterfaceSpriteRendererInfo
  &NiagaraDataInterfaceArrayFloat
  &NiagaraDataInterfaceSkeletalMesh
 ]
 usedDynamicInputs []
 usedModules [
  &self:SystemSpawnScript.SystemScriptSource.SystemScriptGraph.NiagaraNodeAssignment_0.SetVariables_#1
  &SystemState
 ]
}
emitters {
 NE_VC_Boost_Rings {
  bEnabled=false
  emitterUpdateScript [
   &EmitterState(Inactive Response=::NewEnumerator1, Life Cycle Mode=::NewEnumerator1, Loop Duration=0.05, UseLoopDelay=false)
   &SpawnBurst_Instantaneous(Spawn Count=2, Use Spawn Probability=false, UseLoopCountLimit=false)
   &VC_FilteredSocketCount() {enabled=false}
   &SpawnRate(SpawnRate=25, Use Spawn Probability=false) {enabled=false}
  ]
  particleSpawnScript [
   &VC_PaintedVariantsOverride(
    BurntSienna=<rgba 0.7 0.0233 0 1>
    Cobalt=<rgba 0 0.1605 1.07 1>
    Crimson=<rgba 0.4635 0 0 1>
    DefaultColor=<rgba 200 0 6.6665 1>
    ForestGreen=<rgba 0 0.5 0 1>
    Grey=<rgba 0.21 0.21 0.21 1>
    Lime=<rgba 0.2845 0.5885 0 1>
    Orange=<rgba 4 0.0667 0 1>
    Pink=<rgba 1 0 0.1333 1>
    Purple=<rgba 0.2083 0 0.5 1>
    Saffron=<rgba 0.055 0.0458 0 1>
    SkyBlue=<rgba 0 0.0292 0.035 1>
    TitaniumWhite=<rgba 0.25 0.25 0.25 1>
   )
   &InitializeParticle(
    Color=$Particles.Color
    Lifetime=&RandomRangeFloat(Minimum=0.1, Maximum=0.1)
    Mesh Scale Mode=::NewEnumerator3
    Mesh Uniform Scale=0.25
    Position Offset=&AddVectorToPosition(
     Position=&AddVectorToPosition(Position=<-75 0 0>, Vector=$User.SingleBoostPos)
     Vector=&Multiply_VectorByFloat(Vector=$User.DTCarOffset, Float=1)
    )
    Sprite Rotation Mode=::NewEnumerator2
    Sprite Size Mode=::NewEnumerator3
    UsePositionOffset=true
    Write Material Random=true
   )
   &PartitionParticles(Number Of Partitions=$Emitter.PortCount) {enabled=false}
   &SkeletalMeshLocation(
    Bone / Socket Index=$Particles.PartitionParticles.Partition
    Bone / Socket Sampling Mode=::NewEnumerator5
    MaskSpawnGroup=false
    Mesh Sampling Type=::NewEnumerator6
    Skeletal Mesh=$System.SkeletalMesh
   ) {enabled=false}
   &AddVelocity(Velocity=<-1500 0 0>) {enabled=false}
   &InitialMeshOrientation(
    Mesh Orientation Mode=::NewEnumerator7
    Rotation=&RandomRangeVector(Minimum=<-1 0 -1>, Maximum=<1 0 1>)
    Use Rotation Vector=true
   )
   &ShapeLocation(MaskSpawnGroup=false, Override Seed=false, Sphere Radius=10, UseOwnerScale=true)
   &CameraOffset(Camera Offset Amount=10) {enabled=false}
  ]
  particleUpdateScript [
   &ParticleState()
   &ScaleMeshSize(Scale Factor=&Multiply_VectorByFloat(
     Vector=<1 1 1>
     Float=&FloatFromCurve(FloatCurve=DataInterface() {propertyValues {Curve {keys [
          {arriveTangent=18.6511, interpMode=RCIM_Cubic, leaveTangent=18.6511, tangentMode=RCTM_User, time=0, value=2}
          {arriveTangent=-0.2829, interpMode=RCIM_Cubic, leaveTangent=-0.2829, tangentMode=RCTM_User, time=0.15, value=4}
          <at 1 5 RCIM_Cubic>
         ]}}})
    ))
   &UpdateMeshOrientation(
    Rotation Rate=&SelectFloatFromArray(
     Array Sampling Mode=::NewEnumerator0
     Direct Array Index=?
     Float Selection Array=DataInterface() {propertyValues {FloatData [-5,5]}}
     Random Seed=0
    )
    Rotation Vector=<1 1 1>
   )
   &ScaleVelocity(Velocity Scale=&Multiply_VectorByFloat(
     Vector=<1 1 1>
     Float=&FloatFromCurve(FloatCurve=DataInterface() {propertyValues {Curve {keys [<at 0 32>,<at 0.05 0.2 RCIM_Cubic>]}}})
    )) {enabled=false}
   &SolveForcesAndVelocity(Clamp Velocity=false, Limit Acceleration=false)
   &ScaleColor(
    Scale Alpha=&FloatFromCurve(FloatCurve=DataInterface() {propertyValues {Curve {keys [{arriveTangent=-0.1607, interpMode=RCIM_Cubic, leaveTangent=-0.1607, tangentMode=RCTM_User, time=0, value=1}, <at 1 0 RCIM_Cubic>]}}})
    Scale RGB=&Multiply_VectorByFloat(
     Vector=<1 1 1>
     Float=&MakeCustomFloatFromBool(Bool=$User.bIsPainted, True Float=$User.EmitterForce, False Float=1)
    )
    ScaleA=true
    ScaleRGB=true
   )
   &DynamicMaterialParameters(
    Index 0 Param 1=&FloatFromCurve(
     FloatCurve=DataInterface() {propertyValues {Curve {keys [
         {arriveTangent=-0.0976, interpMode=RCIM_Cubic, leaveTangent=-0.0976, tangentMode=RCTM_User, time=0, value=0.5}
         {arriveTangent=0.054, interpMode=RCIM_Cubic, leaveTangent=0.054, tangentMode=RCTM_User, time=0.35, value=0}
         <at 1 -1 RCIM_Cubic>
        ]}}}
     Scale Curve=-1
    )
    Index 0 Param 2=9.7546
    Index 0 Param 3=&RandomRangeFloat(Minimum=1, Maximum=1)
    Index 0 Param 4=&RandomRangeFloat(Minimum=0.25, Maximum=0.35)
    Module.Param0WriteEnabled=true
    Module.Param1WriteEnabled=true
    Module.Param2WriteEnabled=true
    Module.Param3WriteEnabled=true
   )
  ]
  rendererClasses [&NiagaraMeshRendererProperties, &NiagaraSpriteRendererProperties]
  renderers [
   {rendererClass=&NiagaraMeshRendererProperties, rendererIndex=0}
   {rendererClass=&NiagaraSpriteRendererProperties, rendererIndex=1}
  ]
 }
 NE_VC_Boost_Rings001 {
  emitterUpdateScript [
   &EmitterState(
    Inactive Response=::NewEnumerator1
    Life Cycle Mode=::NewEnumerator1
    Loop Behavior=::NewEnumerator1
    Loop Duration Mode=::NewEnumerator1
    UseLoopDelay=false
   )
   &SpawnRate(
    SpawnGroup=2
    SpawnRate=&MakeCustomFloatFromBool(Bool=$User.bIsFrontendPreview, True Float=20, False Float=20)
    Use Spawn Probability=false
   ) {module=SpawnRate001}
   &SpawnBurst_Instantaneous(Spawn Count=2, Use Spawn Probability=false, UseLoopCountLimit=false)
   &SpawnBurst_Instantaneous(Spawn Group=1, Use Spawn Probability=false, UseLoopCountLimit=false) {module=SpawnBurst_Instantaneous001}
   &SpawnRate(SpawnGroup=3, SpawnRate=30, Use Spawn Probability=false) {module=SpawnRate002}
   &VC_FilteredSocketCount() {enabled=false}
   &SpawnRate(SpawnRate=25, Use Spawn Probability=false) {enabled=false}
  ]
  particleSpawnScript [
   &VC_PaintedVariantsOverride(
    BurntSienna=<rgba 0.7 0.0233 0 1>
    Cobalt=<rgba 0 0.1605 1.07 1>
    Crimson=<rgba 0.4635 0 0 1>
    DefaultColor=<rgba 200 0 6.6665 1>
    ForestGreen=<rgba 0 0.5 0 1>
    Grey=<rgba 0.21 0.21 0.21 1>
    Lime=<rgba 0.2845 0.5885 0 1>
    Orange=<rgba 4 0.0667 0 1>
    Pink=<rgba 1 0 0.1333 1>
    Purple=<rgba 0.2083 0 0.5 1>
    R1 Black:LinearColor=<rgba 0.001 0.001 0.001 1>
    R1 BurntSienna:LinearColor=<rgba 1 0.0667 0 1>
    R1 Cobalt:LinearColor=<rgba 0 0.0437 0.46 1>
    R1 DefaultColor=<rgba 200 0 6.6666 1>
    R1 ForestGreen:LinearColor=<rgba 0 0.5 0 1>
    R1 Grey:LinearColor=<rgba 0.08 0.08 0.08 1>
    R1 Lime:LinearColor=<rgba 0.2845 0.5885 0 1>
    R1 Orange:LinearColor=<rgba 2.5 0.4167 0 1>
    R1 Pink:LinearColor=<rgba 0.5 0 0.3083 1>
    R1 Purple:LinearColor=<rgba 0.2945 0 0.6795 1>
    R1 Saffron:LinearColor=<rgba 0.013 0.015 0 1>
    R1 SkyBlue:LinearColor=<rgba 0 0.0208 0.025 1>
    R1 TitaniumWhite:LinearColor=<rgba 0.25 0.25 0.25 1>
    R2 Black:LinearColor=&SelectLinearColorFromArray(
     Array Sampling Mode=::NewEnumerator0
     Direct Array Index=?
     Color Selection Array=DataInterface() {propertyValues {ColorData [<rgba 0 0 0 1>, <rgba 1 1 1 0.1>]}}
     Random Seed=0
    )
    R2 BurntSienna:LinearColor=<rgba 2 0.334 0.04 0.5>
    R2 Cobalt:LinearColor=<rgba 0 0.05 0.3 0.5>
    R2 Crimson:LinearColor=<rgba 0.005 0 0.0001 4>
    R2 DefaultColor:LinearColor=<rgba 100 0 1.6666 1>
    R2 ForestGreen:LinearColor=<rgba 0 0.01 0 12>
    R2 Grey:LinearColor=<rgba 0.2031 0.2031 0.2031 0.2>
    R2 Lime:LinearColor=<rgba 0.0048 0.01 0 15>
    R2 Orange:LinearColor=<rgba 0.0075 0.001 0 20>
    R2 Pink:LinearColor=<rgba 1.1 0.11 0.6415 0.2>
    R2 Purple:LinearColor=<rgba 0.0013 0 0.003 20>
    R2 Saffron:LinearColor=<rgba 0.0458 0.05 0 1>
    R2 SkyBlue:LinearColor=<rgba 0 0.0665 0.1 1>
    R2 TitaniumWhite:LinearColor=<rgba 0.25 0.25 0.25 0.5>
    Saffron=<rgba 0.055 0.0458 0 1>
    SkyBlue=<rgba 0 0.0292 0.035 1>
    TitaniumWhite=<rgba 0.25 0.25 0.25 1>
    Use R1 Painted Overrides=true
    Use R2 Painted Overrides=true
   )
   /SetVariables(Particles.NewFlagTag:NiagaraInt32=&SpawnGroup()) {isSetParameters=true}
   &InitializeParticle(
    Color=$Particles.Color
    Lifetime=/VehicleCosmetics/FX/Niagara/Modules/MergeEmitters/VC_FloatBySpawnGroup(
     SpawnGroup or ExecIndex:NiagaraInt32=0
     CustomInt:NiagaraInt32=0
     MaximumSpawnGroups:NiagaraInt32=4
     Float SG 0:NiagaraFloat=$Emitter.CurrentLoopDuration
     Float SG 1:NiagaraFloat=$Emitter.CurrentLoopDuration
     Float SG 2:NiagaraFloat=&RandomRangeFloat(Minimum=0.5, Maximum=0.75)
     Float SG 3:NiagaraFloat=&RandomRangeFloat(Minimum=0.5, Maximum=0.7)
     Float SG 4:NiagaraFloat=0
     Float SG 5:NiagaraFloat=0
    )
    Mesh Scale Mode=::NewEnumerator3
    Mesh Uniform Scale=&VC_FloatBySpawnGroup(
     SpawnGroup or ExecIndex:NiagaraInt32=0
     CustomInt:NiagaraInt32=0
     MaximumSpawnGroups:NiagaraInt32=2
     Float SG 0:NiagaraFloat=0.4
     Float SG 1:NiagaraFloat=0.3
     Float SG 2:NiagaraFloat=0
     Float SG 3:NiagaraFloat=0
     Float SG 4:NiagaraFloat=0
     Float SG 5:NiagaraFloat=0
    )
    Position Offset=&AddVectorToPosition(
     Position=&AddVectorToPosition(Position=<-75 0 0>, Vector=$User.SingleBoostPos)
     Vector=&Multiply_VectorByFloat(Vector=$User.DTCarOffset, Float=1)
    )
    Sprite Rotation Angle=&RandomRangeFloat(Minimum=-1, Maximum=1)
    Sprite Rotation Mode=::NewEnumerator1
    Sprite Size Mode=::NewEnumerator3
    Uniform Sprite Size=&RandomRangeFloat(
     Minimum=&VC_FloatBySpawnGroup(
      SpawnGroup or ExecIndex:NiagaraInt32=0
      CustomInt:NiagaraInt32=0
      MaximumSpawnGroups:NiagaraInt32=4
      Float SG 0:NiagaraFloat=0
      Float SG 1:NiagaraFloat=0
      Float SG 2:NiagaraFloat=40
      Float SG 3:NiagaraFloat=75
      Float SG 4:NiagaraFloat=0
      Float SG 5:NiagaraFloat=0
     )
     Maximum=&VC_FloatBySpawnGroup(
      SpawnGroup or ExecIndex:NiagaraInt32=0
      CustomInt:NiagaraInt32=0
      MaximumSpawnGroups:NiagaraInt32=4
      Float SG 0:NiagaraFloat=0
      Float SG 1:NiagaraFloat=0
      Float SG 2:NiagaraFloat=60
      Float SG 3:NiagaraFloat=80
      Float SG 4:NiagaraFloat=0
      Float SG 5:NiagaraFloat=0
     )
    )
    UsePositionOffset=true
    Write Material Random=true
   )
   &PartitionParticles(Number Of Partitions=$Emitter.PortCount) {enabled=false}
   &SkeletalMeshLocation(
    Bone / Socket Index=$Particles.PartitionParticles.Partition
    Bone / Socket Sampling Mode=::NewEnumerator5
    MaskSpawnGroup=false
    Mesh Sampling Type=::NewEnumerator6
    Skeletal Mesh=$System.SkeletalMesh
   ) {enabled=false}
   &AddVelocity(
    Velocity=/VehicleCosmetics/FX/Niagara/Modules/MergeEmitters/VC_VectorBySpawnGroup(
     SpawnGroup or ExecIndex:NiagaraInt32=0
     CustomInt:NiagaraInt32=0
     MaximumSpawnGroups:NiagaraInt32=3
     Vector Group 0:Vector3f=<0 0 0>
     Vector Group 1:Vector3f=<0 0 0>
     Vector Group 2:Vector3f=&VectorFromCurve(VectorCurve=DataInterface() {propertyValues {
        XCurve {keys [<at -0.0015 0.9891>,<at 0.4877 0.2382>]}
        YCurve {keys [<at -0.0015 1.0034>,<at 0.4877 0.3164>]}
        ZCurve {keys [<at -0.0015 1.0099>,<at 0.4877 0.5087>]}
       }})
     Vector Group 3:Vector3f=<0 0 0>
     Vector Group 4:Vector3f=<0 0 0>
     Vector Group 5:Vector3f=<0 0 0>
    )
    Velocity Speed Scale=3
   )
   &InitialMeshOrientation(
    Mesh Orientation Mode=::NewEnumerator7
    Rotation=&RandomRangeVector(Minimum=<-1 0 -1>, Maximum=<1 0 1>)
    Use Rotation Vector=true
   )
   &ShapeLocation(
    MaskSpawnGroup=false
    Override Seed=false
    Sphere Radius=&VC_FloatBySpawnGroup(
     SpawnGroup or ExecIndex:NiagaraInt32=0
     CustomInt:NiagaraInt32=0
     MaximumSpawnGroups:NiagaraInt32=4
     Float SG 0:NiagaraFloat=10
     Float SG 1:NiagaraFloat=0
     Float SG 2:NiagaraFloat=50
     Float SG 3:NiagaraFloat=15
     Float SG 4:NiagaraFloat=0
     Float SG 5:NiagaraFloat=0
    )
    UseOwnerScale=true
   )
   &AddVelocity(
    Cone Angle=360
    Module.UseConeFalloff=false
    Module.UseVelDistribution=false
    Override Seed=false
    Velocity Mode=::NewEnumerator2
    Velocity Speed=&VC_FloatBySpawnGroup(
     SpawnGroup or ExecIndex:NiagaraInt32=0
     CustomInt:NiagaraInt32=0
     MaximumSpawnGroups:NiagaraInt32=4
     Float SG 0:NiagaraFloat=0
     Float SG 1:NiagaraFloat=0
     Float SG 2:NiagaraFloat=0
     Float SG 3:NiagaraFloat=&FloatFromCurve(FloatCurve=DataInterface() {propertyValues {Curve {keys [<at 0 1>,<at 1 0>]}}}, Scale Curve=25)
     Float SG 4:NiagaraFloat=0
     Float SG 5:NiagaraFloat=0
    )
   ) {module=AddVelocity001}
  ]
  particleUpdateScript [
   &ParticleState()
   &ScaleMeshSize(Scale Factor=&VC_VectorBySpawnGroup(
     SpawnGroup or ExecIndex:NiagaraInt32=0
     CustomInt:NiagaraInt32=0
     MaximumSpawnGroups:NiagaraInt32=2
     Vector Group 0:Vector3f=&Multiply_VectorByFloat(
      Vector=<1.2 1.2 1.2>
      Float=&FloatFromCurve(FloatCurve=DataInterface() {propertyValues {Curve {keys [
           {arriveTangent=18.6511, interpMode=RCIM_Cubic, leaveTangent=18.6511, tangentMode=RCTM_User, time=0, value=2}
           {arriveTangent=-0.2829, interpMode=RCIM_Cubic, leaveTangent=-0.2829, tangentMode=RCTM_User, time=0.15, value=4}
           <at 1 5 RCIM_Cubic>
          ]}}})
     )
     Vector Group 1:Vector3f=<1 1 1>
     Vector Group 2:Vector3f=<0 0 0>
     Vector Group 3:Vector3f=<0 0 0>
     Vector Group 4:Vector3f=<0 0 0>
     Vector Group 5:Vector3f=<0 0 0>
    ))
   &UpdateMeshOrientation(
    Rotation Rate=&VC_FloatBySpawnGroup(
     SpawnGroup or ExecIndex:NiagaraInt32=0
     CustomInt:NiagaraInt32=0
     MaximumSpawnGroups:NiagaraInt32=2
     Float SG 0:NiagaraFloat=&SelectFloatFromArray(
      Array Sampling Mode=::NewEnumerator0
      Direct Array Index=?
      Float Selection Array=DataInterface() {propertyValues {FloatData [-5,5]}}
      Random Seed=0
     )
     Float SG 1:NiagaraFloat=0
     Float SG 2:NiagaraFloat=0
     Float SG 3:NiagaraFloat=0
     Float SG 4:NiagaraFloat=0
     Float SG 5:NiagaraFloat=0
    )
    Rotation Vector=<1 1 1>
   )
   &Drag(
    Drag=&VC_FloatBySpawnGroup(
     SpawnGroup or ExecIndex:NiagaraInt32=0
     CustomInt:NiagaraInt32=0
     MaximumSpawnGroups:NiagaraInt32=4
     Float SG 0:NiagaraFloat=0
     Float SG 1:NiagaraFloat=0
     Float SG 2:NiagaraFloat=0
     Float SG 3:NiagaraFloat=&FloatFromCurve(
      FloatCurve=DataInterface() {propertyValues {Curve {keys [{tangentMode=RCTM_User, time=0, value=0}, {tangentMode=RCTM_User, time=1, value=1}]}}}
      Scale Curve=5
     )
     Float SG 4:NiagaraFloat=0
     Float SG 5:NiagaraFloat=0
    )
    Rotational Drag:NiagaraFloat=&VC_FloatBySpawnGroup(
     SpawnGroup or ExecIndex:NiagaraInt32=0
     CustomInt:NiagaraInt32=0
     MaximumSpawnGroups:NiagaraInt32=4
     Float SG 0:NiagaraFloat=0
     Float SG 1:NiagaraFloat=0
     Float SG 2:NiagaraFloat=0
     Float SG 3:NiagaraFloat=1
     Float SG 4:NiagaraFloat=0
     Float SG 5:NiagaraFloat=0
    )
    Use Linear Drag=true
    Use Rotational Drag=true
   )
   &SolveForcesAndVelocity(Clamp Velocity=false, Limit Acceleration=false)
   &SubUVAnimation(
    End Frame Range Override=45
    Output Alternate Renderer Binding=true
    "Sprite Renderer "=DataInterface() {propertyValues {SpriteRenderer=&self:NE_VC_Boost_Rings001_0.NiagaraSpriteRendererProperties_1}}
    Start Frame Offset=10
    UseEndFrame=false
    UseStartFrame=false
   ) {module=SubUVAnimation001}
   &ScaleColor(
    Scale Alpha=&FloatFromCurve(FloatCurve=DataInterface() {propertyValues {Curve {keys [{arriveTangent=-0.1607, interpMode=RCIM_Cubic, leaveTangent=-0.1607, tangentMode=RCTM_User, time=0, value=1}, <at 1 0 RCIM_Cubic>]}}})
    Scale RGB=&Multiply_VectorByFloat(
     Vector=<1 1 1>
     Float=&MakeCustomFloatFromBool(Bool=$User.bIsPainted, True Float=$User.EmitterForce, False Float=1)
    )
    ScaleA=true
    ScaleRGB=true
   )
   &DynamicMaterialParameters(
    Index 0 Param 1=&FloatFromCurve(
     FloatCurve=DataInterface() {propertyValues {Curve {keys [
         {arriveTangent=-0.0976, interpMode=RCIM_Cubic, leaveTangent=-0.0976, tangentMode=RCTM_User, time=0, value=0.5}
         {arriveTangent=0.054, interpMode=RCIM_Cubic, leaveTangent=0.054, tangentMode=RCTM_User, time=0.35, value=0}
         <at 1 -1 RCIM_Cubic>
        ]}}}
     Scale Curve=-1
    )
    Index 0 Param 2=9.7546
    Index 0 Param 3=&RandomRangeFloat(Minimum=1, Maximum=1)
    Index 0 Param 4=&RandomRangeFloat(Minimum=0.25, Maximum=0.35)
    Module.Param0WriteEnabled=true
    Module.Param1WriteEnabled=true
    Module.Param2WriteEnabled=true
    Module.Param3WriteEnabled=true
   )
   /SetVariables(
    Particles.DynamicLightning:Vector4f=&MakeVector4(
     X=&RandomRangeFloat(Minimum=0, Maximum=1)
     Y=&RandomRangeFloat(Minimum=10, Maximum=15)
     Z=0
     W=&RandomRangeFloat(Minimum=0, Maximum=0.5)
    )
    Particles.DynamicSmoke:Vector4f=&MakeVector4(
     X=&FloatFromCurve(FloatCurve=DataInterface() {propertyValues {Curve {keys [<at 0 1>,<at 1 32>]}}})
     Y=1
     Z=1
     W=0
    )
    Particles.Dynamic_Core:Vector4f=&MakeVector4(
     X=&FloatFromCurve(FloatCurve=DataInterface() {propertyValues {Curve {keys [<at 0 1 RCIM_Cubic>,<at 1 0 RCIM_Cubic>]}}})
     Y=7.0957
     Z=0.8
     W=5
    )
    Particles.R1 Color:LinearColor=&MultiplyLinearColor(
     Linear Color=$Particles.R1 Color
     Scale Factor=&MakeVector4(
      X=&Multiply_Float(A=1, B=&MakeCustomFloatFromBool(Bool=$User.bIsPainted, True Float=$User.EmitterForce, False Float=1))
      Y=&Multiply_Float(A=1, B=&MakeCustomFloatFromBool(Bool=$User.bIsPainted, True Float=$User.EmitterForce, False Float=1))
      Z=&Multiply_Float(A=1, B=&MakeCustomFloatFromBool(Bool=$User.bIsPainted, True Float=$User.EmitterForce, False Float=1))
      W=1
     )
    )
    Particles.R2 Color:LinearColor=&MultiplyLinearColor(
     Linear Color=$Particles.R2 Color
     Scale Factor=&MakeVector4(
      X=&Multiply_Float(A=1, B=&MakeCustomFloatFromBool(Bool=$User.bIsPainted, True Float=1.5, False Float=1))
      Y=&Multiply_Float(A=1, B=&MakeCustomFloatFromBool(Bool=$User.bIsPainted, True Float=1.5, False Float=1))
      Z=&Multiply_Float(A=1, B=&MakeCustomFloatFromBool(Bool=$User.bIsPainted, True Float=1.5, False Float=1))
      W=&FloatFromCurve(FloatCurve=DataInterface() {propertyValues {Curve {keys [
           <at 0 2 RCIM_Cubic>
           {arriveTangent=0.1684, interpMode=RCIM_Cubic, leaveTangent=0.1684, tangentMode=RCTM_User, time=0.25, value=0.9435}
           {arriveTangent=0.0155, interpMode=RCIM_Cubic, leaveTangent=0.0155, tangentMode=RCTM_User, time=0.5528, value=0.35}
           <at 1 0 RCIM_Cubic>
          ]}}})
     )
    )
    Particles.R3 Color:LinearColor=&MultiplyLinearColor(
     Linear Color=<rgba 0.05 0.05 0.05 1>
     Scale Factor=&MakeVector4(
      X=&MakeCustomFloatFromBool(Bool=$User.bIsPainted, True Float=1, False Float=1.1)
      Y=&MakeCustomFloatFromBool(Bool=$User.bIsPainted, True Float=1, False Float=1.1)
      Z=&MakeCustomFloatFromBool(Bool=$User.bIsPainted, True Float=1, False Float=1.1)
      W=&FloatFromCurve(FloatCurve=DataInterface() {propertyValues {Curve {keys [
           {arriveTangent=-1.0526, interpMode=RCIM_Cubic, time=0, value=1}
           {arriveTangent=0.0155, interpMode=RCIM_Cubic, leaveTangent=0.0155, tangentMode=RCTM_User, time=0.2, value=1}
           <at 0.95 0 RCIM_Cubic>
          ]}}})
     )
    )
   ) {isSetParameters=true}
   &CameraOffset(Camera Offset Amount=&VC_FloatBySpawnGroup(
     SpawnGroup or ExecIndex:NiagaraInt32=0
     CustomInt:NiagaraInt32=0
     MaximumSpawnGroups:NiagaraInt32=2
     Float SG 0:NiagaraFloat=0
     Float SG 1:NiagaraFloat=8
     Float SG 2:NiagaraFloat=0
     Float SG 3:NiagaraFloat=0
     Float SG 4:NiagaraFloat=0
     Float SG 5:NiagaraFloat=0
    )) {module=CameraOffset001}
   &ScaleSpriteSize(
    Scale Sprite Size Mode=::NewEnumerator0
    Uniform Scale Factor=&VC_FloatBySpawnGroup(
     SpawnGroup or ExecIndex:NiagaraInt32=0
     CustomInt:NiagaraInt32=0
     MaximumSpawnGroups:NiagaraInt32=4
     Float SG 0:NiagaraFloat=0
     Float SG 1:NiagaraFloat=0
     Float SG 2:NiagaraFloat=&FloatFromCurve(FloatCurve=DataInterface() {propertyValues {Curve {keys [{tangentMode=RCTM_User, time=0, value=1}, <at 0.15 1.5>, {tangentMode=RCTM_User, time=1, value=3}]}}})
     Float SG 3:NiagaraFloat=&FloatFromCurve(FloatCurve=DataInterface() {propertyValues {Curve {keys [{tangentMode=RCTM_User, time=0, value=2}, <at 0.15 4>, {tangentMode=RCTM_User, time=0.8, value=5}]}}})
     Float SG 4:NiagaraFloat=0
     Float SG 5:NiagaraFloat=0
    )
   )
   &SubUVAnimation(
    End Frame Range Override=8
    Output Alternate Renderer Binding=true
    "Sprite Renderer "=DataInterface() {propertyValues {SpriteRenderer=&self:NE_VC_Boost_Rings001_0.NiagaraSpriteRendererProperties_0}}
    UseEndFrame=true
    UseStartFrame=false
   )
  ]
  rendererClasses [&NiagaraMeshRendererProperties, &NiagaraSpriteRendererProperties]
  renderers [
   {rendererClass=&NiagaraMeshRendererProperties, rendererIndex=0}
   {rendererClass=&NiagaraMeshRendererProperties, rendererIndex=1}
   {rendererClass=&NiagaraSpriteRendererProperties, rendererIndex=2}
   {rendererClass=&NiagaraSpriteRendererProperties, rendererIndex=3}
  ]
 }
 NE_VC_Boost_SMOKECORE {
  bEnabled=false
  emitterUpdateScript [
   &EmitterState(Life Cycle Mode=::NewEnumerator1, Loop Duration=2.25, UseLoopDelay=false)
   &VC_FilteredSocketCount() {enabled=false}
   &SpawnRate(SpawnRate=25, Use Spawn Probability=false)
  ]
  particleSpawnScript [
   &VC_PaintedVariantsOverride(DefaultColor=<rgba 0.05 0.05 0.05 1>) {enabled=false}
   &InitializeParticle(
    Color=<rgba 0.05 0.05 0.05 1>
    Lifetime=&RandomRangeFloat(Minimum=0.5, Maximum=0.7)
    Position Offset=&AddVectorToPosition(
     Position=&AddVectorToPosition(Position=<-70 350 0>, Vector=$User.SingleBoostPos)
     Vector=&Multiply_VectorByFloat(Vector=$User.DTCarOffset, Float=1)
    )
    Sprite Rotation Angle=&RandomRangeFloat(Minimum=-1, Maximum=1)
    Sprite Rotation Mode=::NewEnumerator1
    Sprite Size Mode=::NewEnumerator3
    Uniform Sprite Size=&RandomRangeFloat(Minimum=&Multiply_Float(A=15, B=5), Maximum=&Multiply_Float(A=20, B=4))
    UsePositionOffset=true
    Write Material Random=true
   )
   &PartitionParticles(Number Of Partitions=$Emitter.PortCount) {enabled=false}
   &SkeletalMeshLocation(
    Bone / Socket Index=$Particles.PartitionParticles.Partition
    Bone / Socket Sampling Mode=::NewEnumerator5
    MaskSpawnGroup=false
    Mesh Sampling Type=::NewEnumerator6
    Skeletal Mesh=$System.SkeletalMesh
   ) {enabled=false}
   &AddVelocity(
    Cone Angle=360
    Module.UseConeFalloff=false
    Module.UseVelDistribution=false
    Override Seed=false
    Velocity Mode=::NewEnumerator2
    Velocity Speed=&FloatFromCurve(FloatCurve=DataInterface() {propertyValues {Curve {keys [<at 0 1>,<at 1 0>]}}}, Scale Curve=25)
   )
   &AddVelocity(Velocity=<100 0 0>, Velocity Speed Scale=-1) {enabled=false, module=AddVelocity001}
   &ShapeLocation(MaskSpawnGroup=false, Override Seed=false, Sphere Radius=15, UseOwnerScale=true)
   &CameraOffset(Camera Offset Amount=15) {enabled=false}
  ]
  particleUpdateScript [
   &ParticleState()
   &ScaleSpriteSize(Uniform Curve Sprite Scale=DataInterface() {propertyValues {Curve {keys [{tangentMode=RCTM_User, time=0, value=2}, <at 0.15 4>, {tangentMode=RCTM_User, time=0.8, value=5}]}}})
   &SpriteRotationRate(Rotation Rate=&RandomRangeFloat(Minimum=-1, Maximum=0)) {enabled=false}
   &AccelerationForce(Acceleration=&RandomRangeVector(Minimum=<0 0 -100>, Maximum=<0 0 -75>)) {enabled=false}
   &Drag(
    Drag=&FloatFromCurve(
     FloatCurve=DataInterface() {propertyValues {Curve {keys [{tangentMode=RCTM_User, time=0, value=0}, {tangentMode=RCTM_User, time=1, value=1}]}}}
     Scale Curve=5
    )
    Use Linear Drag=true
    Use Rotational Drag=true
   )
   &SolveForcesAndVelocity(Clamp Velocity=false, Limit Acceleration=false)
   &SubUVAnimation(
    End Frame Range Override=45
    "Sprite Renderer "=DataInterface() {propertyValues {SpriteRenderer=&self:NE_VC_Boost_SMOKECORE_1.NiagaraSpriteRendererProperties_0}}
    Start Frame Offset=10
    UseEndFrame=false
    UseStartFrame=false
   )
   &DynamicMaterialParameters(
    Index 0 Param 1=&FloatFromCurve(FloatCurve=DataInterface() {propertyValues {Curve {keys [<at 0 1>,<at 1 32>]}}})
    Index 0 Param 2=1
    Index 0 Param 3=1
    Module.Index0Param0Write=true
    Module.Index0Param1Write=true
    Module.Index0Param2Write=true
    Module.Index0Param3Write=true
   )
   &ScaleColor(
    Scale Alpha=&FloatFromCurve(FloatCurve=DataInterface() {propertyValues {Curve {keys [
         {arriveTangent=-1.0526, interpMode=RCIM_Cubic, time=0, value=1}
         {arriveTangent=0.0155, interpMode=RCIM_Cubic, leaveTangent=0.0155, tangentMode=RCTM_User, time=0.2, value=1}
         <at 0.95 0 RCIM_Cubic>
        ]}}})
    Scale RGB=&Multiply_VectorByFloat(
     Vector=<1 1 1>
     Float=&MakeCustomFloatFromBool(Bool=$User.bIsFrontendPreview, True Float=1, False Float=3)
    )
    ScaleA=true
    ScaleRGB=true
   )
  ]
  rendererClasses [&NiagaraSpriteRendererProperties]
  renderers [{rendererClass=&NiagaraSpriteRendererProperties}]
 }
 NE_VC_Boost_SMOKECORE001 {
  bEnabled=false
  emitterUpdateScript [
   &EmitterState(Life Cycle Mode=::NewEnumerator1, Loop Duration=2.25, UseLoopDelay=false)
   &VC_FilteredSocketCount() {enabled=false}
   &SpawnRate(
    SpawnRate=&MakeCustomFloatFromBool(Bool=$User.bIsFrontendPreview, True Float=50, False Float=100)
    Use Spawn Probability=false
   )
  ]
  particleSpawnScript [
   &VC_PaintedVariantsOverride(DefaultColor=<rgba 0.06 0.06 0.06 1>) {enabled=false}
   &InitializeParticle(
    Color=<rgba 0.06 0.06 0.06 1>
    Lifetime=&MakeCustomFloatFromBool(Bool=$User.bIsFrontendPreview, True Float=0.4, False Float=0.6)
    Position Offset=&AddVectorToPosition(
     Position=&AddVectorToPosition(Position=<-70 350 0>, Vector=$User.SingleBoostPos)
     Vector=&Multiply_VectorByFloat(Vector=$User.DTCarOffset, Float=1)
    )
    Position Offset Coordinate Space=::Local
    Sprite Rotation Mode=::NewEnumerator2
    Sprite Size Mode=::NewEnumerator3
    Uniform Sprite Size=&RandomRangeFloat(Minimum=&Multiply_Float(A=15, B=5), Maximum=&Multiply_Float(A=20, B=6))
    UsePositionOffset=true
    Write Material Random=true
   )
   &PartitionParticles(Number Of Partitions=$Emitter.PortCount) {enabled=false}
   &SkeletalMeshLocation(
    Bone / Socket Index=$Particles.PartitionParticles.Partition
    Bone / Socket Sampling Mode=::NewEnumerator5
    MaskSpawnGroup=false
    Mesh Sampling Type=::NewEnumerator6
    Skeletal Mesh=$System.SkeletalMesh
   ) {enabled=false}
   &AddVelocity(
    Cone Angle=360
    Module.UseConeFalloff=false
    Module.UseVelDistribution=false
    Override Seed=false
    Velocity Mode=::NewEnumerator2
    Velocity Speed=&FloatFromCurve(FloatCurve=DataInterface() {propertyValues {Curve {keys [<at 0 1>,<at 1 0>]}}}, Scale Curve=25)
   ) {module=AddVelocity001}
   &AddVelocity(Velocity=&RandomRangeVector(Minimum=<2000 0 0>, Maximum=<2500 0 0>), Velocity Speed Scale=-1)
   &ShapeLocation(MaskSpawnGroup=false, Override Seed=false, Sphere Radius=5, UseOwnerScale=true) {enabled=false}
   &CameraOffset(Camera Offset Amount=-50)
  ]
  particleUpdateScript [
   &ParticleState()
   &ScaleSpriteSize(
    Scale Sprite Size Mode=::NewEnumerator0
    Uniform Scale Factor=&MakeCustomFloatFromBool(
     Bool=$User.bIsFrontendPreview
     True Float=&FloatFromCurve(FloatCurve=DataInterface() {propertyValues {Curve {keys [{tangentMode=RCTM_User, time=0, value=2}, <at 0.15 3>, {tangentMode=RCTM_User, time=0.5, value=4}, <at 0.7 4>, <at 1 0>]}}})
     False Float=&FloatFromCurve(FloatCurve=DataInterface() {propertyValues {Curve {keys [{tangentMode=RCTM_User, time=0, value=2}, <at 0.15 3>, <at 1 8>]}}})
    )
   )
   &SpriteRotationRate(Rotation Rate=&RandomRangeFloat(Minimum=-100, Maximum=100))
   &AccelerationForce(Acceleration=&RandomRangeVector(Minimum=<0 0 -100>, Maximum=<0 0 -75>)) {enabled=false}
   &InheritSourceMovement(
    Applied Position Scale=&Multiply_VectorByFloat(Vector=<1 1 1>, Float=1)
    ApplyForce=false
    ApplyPosition=true
    ApplyPositionFalloff=false
   )
   &SolveForcesAndVelocity(Clamp Velocity=false, Limit Acceleration=false)
   &SubUVAnimation(
    End Frame Range Override=45
    "Sprite Renderer "=DataInterface() {propertyValues {SpriteRenderer=&self:NE_VC_Boost_SMOKECORE001_3.NiagaraSpriteRendererProperties_0}}
    Start Frame Offset=10
    UseEndFrame=true
    UseStartFrame=false
   )
   &DynamicMaterialParameters(
    Index 0 Param 1=&FloatFromCurve(FloatCurve=DataInterface() {propertyValues {Curve {keys [<at 0 1>,<at 1 32>]}}})
    Index 0 Param 2=1
    Index 0 Param 3=1
    Module.Index0Param0Write=true
    Module.Index0Param1Write=true
    Module.Index0Param2Write=true
    Module.Index0Param3Write=true
   )
   &ScaleColor(
    Scale Alpha=&FloatFromCurve(FloatCurve=DataInterface() {propertyValues {Curve {keys [<at 0 1 RCIM_Cubic>, {arriveTangentWeight=0.4418, interpMode=RCIM_Cubic, tangentMode=RCTM_User, time=1, value=0}]}}})
    Scale RGB=&Multiply_VectorByFloat(
     Vector=<1 1 1>
     Float=&MakeCustomFloatFromBool(Bool=$User.bIsFrontendPreview, True Float=1, False Float=3)
    )
    ScaleA=true
    ScaleRGB=true
   )
  ]
  rendererClasses [&NiagaraSpriteRendererProperties]
  renderers [{rendererClass=&NiagaraSpriteRendererProperties}]
 }
 NE_VC_Boost_SMOKECORE002 {
  emitterUpdateScript [
   &EmitterState(Life Cycle Mode=::NewEnumerator1, Loop Duration=2.25, UseLoopDelay=false)
   &SpawnRate(
    SpawnGroup=1
    SpawnRate=&MakeCustomFloatFromBool(Bool=$User.bIsFrontendPreview, True Float=80, False Float=100)
    Use Spawn Probability=false
   ) {module=SpawnRate001}
   &VC_FilteredSocketCount() {enabled=false}
   &SpawnRate(
    SpawnRate=&MakeCustomFloatFromBool(Bool=$User.bIsFrontendPreview, True Float=75, False Float=100)
    Use Spawn Probability=false
   )
  ]
  particleSpawnScript [
   &VC_PaintedVariantsOverride(
    DefaultColor=<rgba 0.06 0.06 0.06 1>
    R1 Black:LinearColor=<rgba 0.0001 0.0001 0.0001 1>
    R1 BurntSienna:LinearColor=<rgba 1.5 0.125 0 1>
    R1 Cobalt:LinearColor=<rgba 0 0.0429 0.245 1.1>
    R1 DefaultColor=<rgba 200 0 6.6665 1>
    R1 ForestGreen:LinearColor=<rgba 0 0.5 0.0167 1>
    R1 Grey:LinearColor=<rgba 0.25 0.25 0.25 1>
    R1 Lime:LinearColor=<rgba 0.2658 0.55 0 1.1>
    R1 Orange:LinearColor=<rgba 2 0.0667 0 1>
    R1 Pink:LinearColor=<rgba 5 0 3.3333 1>
    R1 Purple:LinearColor=<rgba 0.1667 0 0.5 1.1>
    R1 Saffron:LinearColor=<rgba 0.0083 0.01 0 1>
    R1 SkyBlue:LinearColor=<rgba 0 0.0083 0.01 1>
    Use R1 Painted Overrides=true
   )
   /SetVariables(Particles.SetNewFlag:NiagaraInt32=&SpawnGroup()) {isSetParameters=true}
   &InitializeParticle(
    Color=<rgba 0.06 0.06 0.06 1>
    Lifetime=&VC_FloatBySpawnGroup(
     SpawnGroup or ExecIndex:NiagaraInt32=0
     CustomInt:NiagaraInt32=0
     MaximumSpawnGroups:NiagaraInt32=2
     Float SG 0:NiagaraFloat=&MakeCustomFloatFromBool(Bool=$User.bIsFrontendPreview, True Float=0.4, False Float=0.6)
     Float SG 1:NiagaraFloat=&MakeCustomFloatFromBool(Bool=$User.bIsFrontendPreview, True Float=0.3, False Float=0.6)
     Float SG 2:NiagaraFloat=0
     Float SG 3:NiagaraFloat=0
     Float SG 4:NiagaraFloat=0
     Float SG 5:NiagaraFloat=0
    )
    Mesh Scale Max=&VC_VectorBySpawnGroup(
     SpawnGroup or ExecIndex:NiagaraInt32=0
     CustomInt:NiagaraInt32=0
     MaximumSpawnGroups:NiagaraInt32=2
     Vector Group 0:Vector3f=<0 0 0>
     Vector Group 1:Vector3f=<0.375 0.375 0.375>
     Vector Group 2:Vector3f=<0 0 0>
     Vector Group 3:Vector3f=<0 0 0>
     Vector Group 4:Vector3f=<0 0 0>
     Vector Group 5:Vector3f=<0 0 0>
    )
    Mesh Scale Min=&VC_VectorBySpawnGroup(
     SpawnGroup or ExecIndex:NiagaraInt32=0
     CustomInt:NiagaraInt32=0
     MaximumSpawnGroups:NiagaraInt32=2
     Vector Group 0:Vector3f=<0 0 0>
     Vector Group 1:Vector3f=<0.25 0.25 0.25>
     Vector Group 2:Vector3f=<0 0 0>
     Vector Group 3:Vector3f=<0 0 0>
     Vector Group 4:Vector3f=<0 0 0>
     Vector Group 5:Vector3f=<0 0 0>
    )
    Mesh Scale Mode=::NewEnumerator2
    Position Offset=&AddVectorToPosition(
     Position=&AddVectorToPosition(Position=<-70 0 0>, Vector=$User.SingleBoostPos)
     Vector=&Multiply_VectorByFloat(Vector=$User.DTCarOffset, Float=1)
    )
    Position Offset Coordinate Space=::Local
    Sprite Rotation Mode=::NewEnumerator2
    Sprite Size Mode=::NewEnumerator3
    Uniform Sprite Size=&RandomRangeFloat(Minimum=&Multiply_Float(A=15, B=5), Maximum=&Multiply_Float(A=20, B=6))
    UsePositionOffset=true
    Write Material Random=true
   )
   &PartitionParticles(Number Of Partitions=$Emitter.PortCount) {enabled=false}
   &SkeletalMeshLocation(
    Bone / Socket Index=$Particles.PartitionParticles.Partition
    Bone / Socket Sampling Mode=::NewEnumerator5
    MaskSpawnGroup=false
    Mesh Sampling Type=::NewEnumerator6
    Skeletal Mesh=$System.SkeletalMesh
   ) {enabled=false}
   &AddVelocity(
    Cone Angle=360
    Module.UseConeFalloff=false
    Module.UseVelDistribution=false
    Override Seed=false
    Velocity Mode=::NewEnumerator2
    Velocity Speed=&FloatFromCurve(FloatCurve=DataInterface() {propertyValues {Curve {keys [<at 0 1>,<at 1 0>]}}}, Scale Curve=25)
   ) {module=AddVelocity001}
   &AddVelocity(Velocity=&RandomRangeVector(Minimum=<2000 0 0>, Maximum=<2500 0 0>), Velocity Speed Scale=-1)
   &ShapeLocation(MaskSpawnGroup=false, Override Seed=false, Sphere Radius=5, UseOwnerScale=true) {enabled=false}
   &CameraOffset(Camera Offset Amount=-50)
   &AddVelocity(Velocity=&VC_VectorBySpawnGroup(
     SpawnGroup or ExecIndex:NiagaraInt32=0
     CustomInt:NiagaraInt32=0
     MaximumSpawnGroups:NiagaraInt32=2
     Vector Group 0:Vector3f=<0 0 0>
     Vector Group 1:Vector3f=<-2000 0 0>
     Vector Group 2:Vector3f=<0 0 0>
     Vector Group 3:Vector3f=<0 0 0>
     Vector Group 4:Vector3f=<0 0 0>
     Vector Group 5:Vector3f=<0 0 0>
    )) {module=AddVelocity002}
   &InitialMeshOrientation(
    Mesh Orientation Mode=::NewEnumerator7
    Rotation=&RandomRangeVector(Minimum=<-1 -1 -1>, Maximum=<1 1 1>)
    Use Rotation Vector=true
   )
  ]
  particleUpdateScript [
   &ParticleState()
   &ScaleSpriteSize(
    Scale Sprite Size Mode=::NewEnumerator0
    Uniform Scale Factor=&MakeCustomFloatFromBool(
     Bool=$User.bIsFrontendPreview
     True Float=&FloatFromCurve(FloatCurve=DataInterface() {propertyValues {Curve {keys [{tangentMode=RCTM_User, time=0, value=2}, <at 0.15 3>, {tangentMode=RCTM_User, time=0.5, value=4}, <at 0.7 4>, <at 1 0>]}}})
     False Float=&FloatFromCurve(FloatCurve=DataInterface() {propertyValues {Curve {keys [{tangentMode=RCTM_User, time=0, value=2}, <at 0.15 3>, <at 1 8>]}}})
    )
   )
   &SpriteRotationRate(Rotation Rate=&RandomRangeFloat(Minimum=-100, Maximum=100))
   &InheritSourceMovement(
    Applied Position Scale=&Multiply_VectorByFloat(Vector=<1 1 1>, Float=1)
    ApplyForce=false
    ApplyPosition=true
    ApplyPositionFalloff=false
   ) {module=InheritSourceMovement001}
   &AccelerationForce(Acceleration=&RandomRangeVector(Minimum=<0 0 -100>, Maximum=<0 0 -75>)) {enabled=false}
   &InheritSourceMovement(
    Applied Position Scale=&Multiply_VectorByFloat(Vector=<1 1 1>, Float=1)
    ApplyForce=false
    ApplyPosition=true
    ApplyPositionFalloff=false
   )
   &SolveForcesAndVelocity(Clamp Velocity=false, Limit Acceleration=false)
   &ScaleMeshSize(Scale Factor=&Multiply_VectorByFloat(
     Vector=<1 1 1>
     Float=&VC_FloatBySpawnGroup(
      SpawnGroup or ExecIndex:NiagaraInt32=0
      CustomInt:NiagaraInt32=0
      MaximumSpawnGroups:NiagaraInt32=2
      Float SG 0:NiagaraFloat=0
      Float SG 1:NiagaraFloat=&FloatFromCurve(
       FloatCurve=DataInterface() {propertyValues {Curve {keys [{arriveTangent=18.6511, time=0, value=1}, <at 0.1 1>, <at 0.9 0 RCIM_Constant>]}}}
       Scale Curve=&MakeCustomFloatFromBool(Bool=$User.bIsFrontendPreview, True Float=1, False Float=1.5)
      )
      Float SG 2:NiagaraFloat=0
      Float SG 3:NiagaraFloat=0
      Float SG 4:NiagaraFloat=0
      Float SG 5:NiagaraFloat=0
     )
    ))
   &SubUVAnimation(
    End Frame Range Override=45
    "Sprite Renderer "=DataInterface() {propertyValues {SpriteRenderer=&self:NE_VC_Boost_SMOKECORE002_0.NiagaraSpriteRendererProperties_0}}
    Start Frame Offset=10
    UseEndFrame=true
    UseStartFrame=false
   )
   &DynamicMaterialParameters(
    Index 0 Param 1=&FloatFromCurve(FloatCurve=DataInterface() {propertyValues {Curve {keys [<at 0 1>,<at 1 32>]}}})
    Index 0 Param 2=1
    Index 0 Param 3=1
    Module.Index0Param0Write=true
    Module.Index0Param1Write=true
    Module.Index0Param2Write=true
    Module.Index0Param3Write=true
   )
   &ScaleColor(
    Scale Alpha=&FloatFromCurve(FloatCurve=DataInterface() {propertyValues {Curve {keys [<at 0 1 RCIM_Cubic>, {arriveTangentWeight=0.4418, interpMode=RCIM_Cubic, tangentMode=RCTM_User, time=1, value=0}]}}})
    Scale RGB=&Multiply_VectorByFloat(
     Vector=<1 1 1>
     Float=&MakeCustomFloatFromBool(Bool=$User.bIsFrontendPreview, True Float=1, False Float=3)
    )
    ScaleA=true
    ScaleRGB=true
   )
   /SetVariables(
    Particles.DynamicMeshTrail:Vector4f=&MakeVector4(
     X=&FloatFromCurve(
      FloatCurve=DataInterface() {propertyValues {Curve {keys [
          {arriveTangent=-2.2308, interpMode=RCIM_Cubic, leaveTangent=-2.2308, tangentMode=RCTM_User, time=0, value=1}
          {interpMode=RCIM_Cubic, tangentMode=RCTM_User, time=1, value=0}
         ]}}}
      Scale Curve=5
     )
     Y=1
     Z=&RandomRangeFloat(Minimum=0, Maximum=2)
     W=25
    )
    Particles.R1 Color:LinearColor=&MultiplyLinearColor(
     Linear Color=$Particles.R1 Color
     Scale Factor=&MakeVector4(
      X=&Multiply_Float(A=1, B=&MakeCustomFloatFromBool(Bool=$User.bIsPainted, True Float=$User.EmitterForce, False Float=1))
      Y=&Multiply_Float(A=1, B=&MakeCustomFloatFromBool(Bool=$User.bIsPainted, True Float=$User.EmitterForce, False Float=1))
      Z=&Multiply_Float(A=1, B=&MakeCustomFloatFromBool(Bool=$User.bIsPainted, True Float=$User.EmitterForce, False Float=1))
      W=&FloatFromCurve(FloatCurve=DataInterface() {propertyValues {Curve {keys [
           {tangentMode=RCTM_User, time=0, value=1}
           {arriveTangent=-0.0405, interpMode=RCIM_Cubic, leaveTangent=-0.0405, tangentMode=RCTM_User, time=0.5, value=1}
           {arriveTangent=-0.0421, interpMode=RCIM_Cubic, leaveTangent=-0.0421, tangentMode=RCTM_User, time=1, value=0}
          ]}}})
     )
    )
   ) {isSetParameters=true}
  ]
  rendererClasses [&NiagaraMeshRendererProperties, &NiagaraSpriteRendererProperties]
  renderers [
   {rendererClass=&NiagaraMeshRendererProperties, rendererIndex=0}
   {rendererClass=&NiagaraSpriteRendererProperties, rendererIndex=1}
  ]
 }
 NE_VC_Boost_Sprite_CORE {
  bEnabled=false
  emitterUpdateScript [
   &EmitterState(Inactive Response=::NewEnumerator1, Life Cycle Mode=::NewEnumerator1, Loop Duration=1, UseLoopDelay=false)
   &SpawnBurst_Instantaneous(Use Spawn Probability=false, UseLoopCountLimit=false)
   &VC_FilteredSocketCount() {enabled=false}
   &SpawnRate(Use Spawn Probability=false) {enabled=false}
  ]
  particleSpawnScript [
   &VC_PaintedVariantsOverride(
    Black=<rgba 0.001 0.001 0.001 1>
    BurntSienna=<rgba 1 0.0667 0 1>
    Cobalt=<rgba 0 0.0437 0.46 1>
    DefaultColor=<rgba 200 0 6.6666 1>
    ForestGreen=<rgba 0 0.5 0 1>
    Grey=<rgba 0.08 0.08 0.08 1>
    Lime=<rgba 0.2845 0.5885 0 1>
    Orange=<rgba 2.5 0.4167 0 1>
    Pink=<rgba 0.5 0 0.3083 1>
    Purple=<rgba 0.2945 0 0.6795 1>
    Saffron=<rgba 0.013 0.015 0 1>
    SkyBlue=<rgba 0 0.0208 0.025 1>
    TitaniumWhite=<rgba 0.25 0.25 0.25 1>
   )
   &InitializeParticle(
    Color=$Particles.Color
    Lifetime=$Emitter.CurrentLoopDuration
    Mesh Scale Mode=::NewEnumerator3
    Mesh Uniform Scale=0.3
    Position Offset=&AddVectorToPosition(
     Position=&AddVectorToPosition(Position=<-75 0 0>, Vector=$User.SingleBoostPos)
     Vector=&Multiply_VectorByFloat(Vector=$User.DTCarOffset, Float=1)
    )
    Sprite Rotation Mode=::NewEnumerator2
    UsePositionOffset=true
    Write Material Random=true
   )
   &PartitionParticles(Number Of Partitions=$Emitter.PortCount) {enabled=false}
   &SkeletalMeshLocation(
    Bone / Socket Index=$Particles.PartitionParticles.Partition
    Bone / Socket Sampling Mode=::NewEnumerator5
    MaskSpawnGroup=false
    Mesh Sampling Type=::NewEnumerator6
    Skeletal Mesh=$System.SkeletalMesh
   ) {enabled=false}
   &AddVelocity(Velocity=<-1500 0 0>) {enabled=false}
  ]
  particleUpdateScript [
   &ParticleState()
   &UpdateMeshOrientation(Rotation Rate=0.5, Rotation Vector=<1 1 1>) {enabled=false}
   &MeshRotationForce(Rotation=<10 10 10>) {enabled=false}
   &SolveForcesAndVelocity(Clamp Velocity=false, Limit Acceleration=false)
   &DynamicMaterialParameters(
    Index 0 Param 1=&FloatFromCurve(FloatCurve=DataInterface() {propertyValues {Curve {keys [<at 0 1 RCIM_Cubic>,<at 1 0 RCIM_Cubic>]}}})
    Index 0 Param 2=7.0957
    Index 0 Param 3=0.8
    Index 0 Param 4=5
    Module.Index0Param0Write=true
    Module.Index0Param1Write=true
    Module.Index0Param2Write=true
    Module.Index0Param3Write=true
   )
   &CameraOffset(Camera Offset Amount=8)
   &ScaleColor(
    Scale RGB=&Multiply_VectorByFloat(
     Vector=<1 1 1>
     Float=&MakeCustomFloatFromBool(Bool=$User.bIsPainted, True Float=$User.EmitterForce, False Float=1)
    )
    ScaleA=true
    ScaleRGB=true
   )
  ]
  rendererClasses [&NiagaraMeshRendererProperties, &NiagaraSpriteRendererProperties]
  renderers [
   {rendererClass=&NiagaraMeshRendererProperties, rendererIndex=0}
   {rendererClass=&NiagaraSpriteRendererProperties, rendererIndex=1}
  ]
 }
 NE_VC_Boost_Sprite_CORE_LIGHTNING {
  bEnabled=false
  emitterUpdateScript [
   &EmitterState(Inactive Response=::NewEnumerator1, Life Cycle Mode=::NewEnumerator1, Loop Delay=0.5, Loop Duration=0.35, UseLoopDelay=true)
   &VC_FilteredSocketCount()
   &SpawnRate(
    SpawnRate=&MakeCustomFloatFromBool(Bool=$User.bIsFrontendPreview, True Float=40, False Float=80)
    Use Spawn Probability=false
   )
  ]
  particleSpawnScript [
   &VC_PaintedVariantsOverride(
    Black=&SelectLinearColorFromArray(
     Array Sampling Mode=::NewEnumerator0
     Direct Array Index=?
     Color Selection Array=DataInterface() {propertyValues {ColorData [<rgba 0 0 0 1>, <rgba 1 1 1 0.1>]}}
     Random Seed=0
    )
    BurntSienna=<rgba 2 0.334 0.04 0.5>
    Cobalt=<rgba 0 0.05 0.3 0.5>
    Crimson=<rgba 0.005 0 0.0001 4>
    DefaultColor=<rgba 100 0 1.6666 1>
    ForestGreen=<rgba 0 0.01 0 12>
    Grey=<rgba 0.2031 0.2031 0.2031 0.2>
    Lime=<rgba 0.0048 0.01 0 15>
    Orange=<rgba 0.0075 0.001 0 20>
    Pink=<rgba 1.1 0.11 0.6415 0.2>
    Purple=<rgba 0.0013 0 0.003 20>
    Saffron=<rgba 0.0458 0.05 0 1>
    SkyBlue=<rgba 0 0.0665 0.1 1>
    TitaniumWhite=<rgba 0.25 0.25 0.25 0.5>
   )
   &InitializeParticle(
    Color=$Particles.Color
    Lifetime=&RandomRangeFloat(Minimum=0.5, Maximum=0.75)
    Position Offset=&AddVectorToPosition(
     Position=&AddVectorToPosition(Position=<-75 0 0>, Vector=$User.SingleBoostPos)
     Vector=&Multiply_VectorByFloat(Vector=$User.DTCarOffset, Float=1)
    )
    Position Offset Coordinate Space=::Local
    Sprite Rotation Angle=&RandomRangeFloat(Minimum=-1, Maximum=1)
    Sprite Rotation Mode=::NewEnumerator1
    Sprite Size Mode=::NewEnumerator3
    Uniform Sprite Size=&RandomRangeFloat(Minimum=40, Maximum=60)
    UsePositionOffset=true
    Write Material Random=true
   )
   &PartitionParticles(Number Of Partitions=$Emitter.PortCount) {enabled=false}
   &SkeletalMeshLocation(
    Bone / Socket Index=$Particles.PartitionParticles.Partition
    Bone / Socket Sampling Mode=::NewEnumerator5
    MaskSpawnGroup=false
    Mesh Sampling Type=::NewEnumerator6
    Skeletal Mesh=$System.SkeletalMesh
   ) {enabled=false}
   &AddVelocity(
    Velocity=&VectorFromCurve(VectorCurve=DataInterface() {propertyValues {
       XCurve {keys [<at -0.0015 0.9891>,<at 0.4877 0.2382>]}
       YCurve {keys [<at -0.0015 1.0034>,<at 0.4877 0.3164>]}
       ZCurve {keys [<at -0.0015 1.0099>,<at 0.4877 0.5087>]}
      }})
    Velocity Speed Scale=3
   )
   &ShapeLocation(MaskSpawnGroup=false, Override Seed=false, Sphere Radius=50, UseOwnerScale=true)
  ]
  particleUpdateScript [
   &ParticleState()
   &SpriteRotationRate(Rotation Rate=&RandomRangeFloat(Minimum=500, Maximum=-500)) {enabled=false}
   &ScaleSpriteSize(Uniform Curve Sprite Scale=DataInterface() {propertyValues {Curve {keys [{tangentMode=RCTM_User, time=0, value=1}, <at 0.15 1.5>, {tangentMode=RCTM_User, time=1, value=3}]}}})
   &SolveForcesAndVelocity(Clamp Velocity=false, Limit Acceleration=false)
   &ScaleColor(
    Scale Alpha=&FloatFromCurve(FloatCurve=DataInterface() {propertyValues {Curve {keys [
         <at 0 2 RCIM_Cubic>
         {arriveTangent=0.1684, interpMode=RCIM_Cubic, leaveTangent=0.1684, tangentMode=RCTM_User, time=0.25, value=0.9435}
         {arriveTangent=0.0155, interpMode=RCIM_Cubic, leaveTangent=0.0155, tangentMode=RCTM_User, time=0.5528, value=0.35}
         <at 1 0 RCIM_Cubic>
        ]}}})
    Scale RGB=&Multiply_VectorByFloat(Vector=<1 1 1>, Float=&MakeCustomFloatFromBool(Bool=$User.bIsPainted, True Float=70, False Float=1))
    ScaleA=true
    ScaleRGB=true
   )
   &DynamicMaterialParameters(
    Index 0 Param 1=&RandomRangeFloat(Minimum=0, Maximum=1)
    Index 0 Param 2=&RandomRangeFloat(Minimum=10, Maximum=15)
    Index 0 Param 4=&RandomRangeFloat(Minimum=0, Maximum=0.5)
    Module.Index0Param0Write=true
    Module.Index0Param1Write=true
    Module.Index0Param2Write=true
    Module.Index0Param3Write=true
   )
   &SubUVAnimation(
    End Frame Range Override=8
    "Sprite Renderer "=DataInterface() {propertyValues {SpriteRenderer=&self:NE_VC_Boost_Sprite_CORE_LIGHTNING_2.NiagaraSpriteRendererProperties_0}}
    UseEndFrame=true
    UseStartFrame=false
   )
  ]
  rendererClasses [&NiagaraSpriteRendererProperties]
  renderers [{rendererClass=&NiagaraSpriteRendererProperties}]
 }
 NE_VC_Boost_Sprite_TRAILLIGHTNING {
  bEnabled=false
  emitterUpdateScript [
   &EmitterState(Life Cycle Mode=::NewEnumerator1, Loop Duration=0.05, UseLoopDelay=false)
   &SpawnBurst_Instantaneous(Use Spawn Probability=false, UseLoopCountLimit=false) {enabled=false}
   &VC_FilteredSocketCount() {enabled=false}
   &SpawnRate(
    SpawnRate=&MakeCustomFloatFromBool(Bool=$User.bIsFrontendPreview, True Float=80, False Float=100)
    Use Spawn Probability=false
   )
  ]
  particleSpawnScript [
   &VC_PaintedVariantsOverride(
    Black=<rgba 0.0001 0.0001 0.0001 1>
    BurntSienna=<rgba 1.5 0.125 0 1>
    Cobalt=<rgba 0 0.0429 0.245 1.1>
    DefaultColor=<rgba 200 0 6.6665 1>
    ForestGreen=<rgba 0 0.5 0.0167 1>
    Grey=<rgba 0.25 0.25 0.25 1>
    Lime=<rgba 0.2658 0.55 0 1.1>
    Orange=<rgba 2 0.0667 0 1>
    Pink=<rgba 5 0 3.3333 1>
    Purple=<rgba 0.1667 0 0.5 1.1>
    Saffron=<rgba 0.0083 0.01 0 1>
    SkyBlue=<rgba 0 0.0083 0.01 1>
   )
   &InitializeParticle(
    Color=$Particles.Color
    Lifetime=&MakeCustomFloatFromBool(Bool=$User.bIsFrontendPreview, True Float=0.3, False Float=0.6)
    Mesh Scale Max=&Multiply_VectorByFloat(Vector=<0.5 0.5 0.5>, Float=0.5)
    Mesh Scale Min=&Multiply_VectorByFloat(Vector=<0.75 0.75 0.75>, Float=0.5)
    Mesh Scale Mode=::NewEnumerator2
    Position Offset=&AddVectorToPosition(
     Position=&AddVectorToPosition(
      Position=/Niagara/DynamicInputs/TypeConversions/MakeCustomPositionFromBool(Bool:NiagaraBool=$User.bIsFrontendPreview, True Position:NiagaraPosition=<-90 350 0>, False Position:NiagaraPosition=<-90 350 25>)
      Vector=$User.SingleBoostPos
     )
     Vector=&Multiply_VectorByFloat(Vector=$User.DTCarOffset, Float=1)
    )
    Position Offset Coordinate Space=::Local
    Sprite Rotation Mode=::NewEnumerator2
    Sprite Size Mode=::NewEnumerator3
    UsePositionOffset=true
    Write Material Random=true
   )
   &PartitionParticles(Number Of Partitions=$Emitter.PortCount) {enabled=false}
   &SkeletalMeshLocation(
    Bone / Socket Index=$Particles.PartitionParticles.Partition
    Bone / Socket Sampling Mode=::NewEnumerator5
    MaskSpawnGroup=false
    Mesh Sampling Type=::NewEnumerator6
    Skeletal Mesh=$System.SkeletalMesh
   ) {enabled=false}
   &AddVelocity(Velocity=<2000 0 0>, Velocity Speed Scale=-1)
   &InitialMeshOrientation(
    Mesh Orientation Mode=::NewEnumerator7
    Rotation=&RandomRangeVector(Minimum=<-1 -1 -1>, Maximum=<1 1 1>)
    Use Rotation Vector=true
   )
  ]
  particleUpdateScript [
   &ParticleState()
   &ScaleMeshSize(Scale Factor=&Multiply_VectorByFloat(
     Vector=<1 1 1>
     Float=&FloatFromCurve(
      FloatCurve=DataInterface() {propertyValues {Curve {keys [{arriveTangent=18.6511, time=0, value=1}, <at 0.1 1>, <at 0.9 0 RCIM_Constant>]}}}
      Scale Curve=&MakeCustomFloatFromBool(Bool=$User.bIsFrontendPreview, True Float=1, False Float=1.5)
     )
    ))
   &InheritSourceMovement(
    Applied Position Scale=&Multiply_VectorByFloat(Vector=<1 1 1>, Float=1)
    ApplyForce=false
    ApplyPosition=true
    ApplyPositionFalloff=false
   )
   &SolveForcesAndVelocity(Clamp Velocity=false, Limit Acceleration=false)
   &DynamicMaterialParameters(
    Index 0 Param 1=&FloatFromCurve(
     FloatCurve=DataInterface() {propertyValues {Curve {keys [
         {arriveTangent=-2.2308, interpMode=RCIM_Cubic, leaveTangent=-2.2308, tangentMode=RCTM_User, time=0, value=1}
         {interpMode=RCIM_Cubic, tangentMode=RCTM_User, time=1, value=0}
        ]}}}
     Scale Curve=5
    )
    Index 0 Param 2=1
    Index 0 Param 3=&RandomRangeFloat(Minimum=0, Maximum=2)
    Index 0 Param 4=25
    Module.Param0WriteEnabled=true
    Module.Param1WriteEnabled=true
    Module.Param2WriteEnabled=true
    Module.Param3WriteEnabled=true
   )
   &ScaleColor(
    Scale Alpha=&FloatFromCurve(FloatCurve=DataInterface() {propertyValues {Curve {keys [
         {tangentMode=RCTM_User, time=0, value=1}
         {arriveTangent=-0.0405, interpMode=RCIM_Cubic, leaveTangent=-0.0405, tangentMode=RCTM_User, time=0.5, value=1}
         {arriveTangent=-0.0421, interpMode=RCIM_Cubic, leaveTangent=-0.0421, tangentMode=RCTM_User, time=1, value=0}
        ]}}})
    Scale RGB=&Multiply_VectorByFloat(
     Vector=<1 1 1>
     Float=&MakeCustomFloatFromBool(Bool=$User.bIsPainted, True Float=$User.EmitterForce, False Float=1)
    )
    ScaleA=true
    ScaleRGB=true
   )
  ]
  rendererClasses [&NiagaraMeshRendererProperties, &NiagaraSpriteRendererProperties]
  renderers [
   {rendererClass=&NiagaraMeshRendererProperties, rendererIndex=0}
   {rendererClass=&NiagaraSpriteRendererProperties, rendererIndex=1}
  ]
 }
}
summary {
 systemName=NS_Boost_Sphenergy
 emitters [
  {
   bEnabled=false
   emitterName=NE_VC_Boost_Sprite_CORE
   rendererClasses [&NiagaraMeshRendererProperties, &NiagaraSpriteRendererProperties]
  }
  {
   bEnabled=false
   emitterName=NE_VC_Boost_Sprite_TRAILLIGHTNING
   rendererClasses [&NiagaraMeshRendererProperties, &NiagaraSpriteRendererProperties]
  }
  {bEnabled=false, emitterName=NE_VC_Boost_Sprite_CORE_LIGHTNING, rendererClasses [&NiagaraSpriteRendererProperties]}
  {
   bEnabled=false
   emitterName=NE_VC_Boost_Rings
   rendererClasses [&NiagaraMeshRendererProperties, &NiagaraSpriteRendererProperties]
  }
  {bEnabled=false, emitterName=NE_VC_Boost_SMOKECORE, rendererClasses [&NiagaraSpriteRendererProperties]}
  {bEnabled=false, emitterName=NE_VC_Boost_SMOKECORE001, rendererClasses [&NiagaraSpriteRendererProperties]}
  {
   emitterName=NE_VC_Boost_Rings001
   rendererClasses [&NiagaraMeshRendererProperties, &NiagaraSpriteRendererProperties]
  }
  {
   emitterName=NE_VC_Boost_SMOKECORE002
   rendererClasses [&NiagaraMeshRendererProperties, &NiagaraSpriteRendererProperties]
  }
 ]
 userVariables [
  User.bIsFrontendPreview {type=NiagaraBool}
  User.bIsPainted {type=NiagaraBool}
  User.ColorVariantID {type=NiagaraInt32}
  User.DTCarOffset {type=Vector3f, description="", defaultValue {struct=Vector3f, value=<1 1 1>}}
  User.EmitterForce {type=NiagaraFloat, defaultValue=70, description=""}
  User.PaintedColor {type=LinearColor, description="", defaultValue {struct=LinearColor, value=<rgba 1 1 1 1>}}
  User.PaintedScalar {type=NiagaraFloat, defaultValue=1, description=""}
  User.PortCount {type=NiagaraInt32, defaultValue=1}
  User.SingleBoostPos {type=Vector3f, description="", defaultValue {struct=Vector3f, value=<0 0 0>}}
 ]
}
