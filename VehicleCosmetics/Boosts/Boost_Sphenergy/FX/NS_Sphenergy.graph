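# Niagara system for the Sphenergy vehicle boost cosmetic, spawning sprite and mesh effects just behind the boost position plus car offset.
# A cone emitter streams short-lived particles, a second pulses bursts of rotating meshes, and a smoke emitter throws drag-slowed particles backward.
# Spawn groups vary shape per layer and set renderer visibility tags.
# Painted-variant overrides pick colors, while per-variant brightness arrays multiplied by VC_Mask_Floats tune glow per context.
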NiagaraSystem /VehicleCosmetics/Boosts/Boost_Sphenergy/FX/NS_Sphenergy
dependencies {
 usedDataInterfaces [
  &NiagaraDataInterfaceCurve
  &NiagaraDataInterfaceArrayFloat
  &NiagaraDataInterfaceSpriteRendererInfo
  &NiagaraDataInterfaceSkeletalMesh
  &NiagaraDataInterfaceArrayFloat3
 ]
 usedDynamicInputs [&SelectVectorFromArray]
 usedModules [
  &self:SystemSpawnScript.SystemScriptSource.SystemScriptGraph.NiagaraNodeAssignment_0.SetVariables_#1
  &self:SystemSpawnScript.SystemScriptSource.SystemScriptGraph.NiagaraNodeAssignment_1.SetVariables_#2
  &self:SystemSpawnScript.SystemScriptSource.SystemScriptGraph.NiagaraNodeAssignment_2.SetVariables_#3
  &SystemState
 ]
}
emitters {
 NE_VC_Cone {
  emitterUpdateScript [
   &EmitterState(Inactive Response=::NewEnumerator1, Life Cycle Mode=::NewEnumerator1, Loop Duration=1, UseLoopDelay=false)
   &SpawnRate(SpawnRate=50, Use Spawn Probability=false)
   &SpawnRate(SpawnGroup=1, SpawnRate=30, Use Spawn Probability=false) {module=SpawnRate001}
   &SpawnRate(Spawn Probability=0.65, SpawnGroup=2, SpawnRate=60, Use Spawn Probability=true) {module=SpawnRate002}
  ]
  particleSpawnScript [
   &VC_PaintedVariantsOverride(
    Black=<rgba 0 0 0 1>
    BurntSienna=<rgba 0.65 0.2784 0.013 1>
    Cobalt=<rgba 0 0.0267 0.4 1>
    Crimson=<rgba 1 0 0 1>
    DefaultColor=<rgba 1 0.0658 0.05 1>
    ForestGreen=<rgba 0 1 0 1>
    Grey=<rgba 0.5 0.5 0.5 1>
    Lime=<rgba 0.1667 0.4 0 1>
    Orange=<rgba 0.3 0.025 0 1>
    Pink=<rgba 0.9 0 0.75 1>
    Purple=<rgba 0.3333 0 0.5 1>
    R1 Black:LinearColor=<rgba 1 1 1 1>
    R1 BurntSienna:LinearColor=<rgba 0.65 0.2784 0.013 1>
    R1 Cobalt:LinearColor=<rgba 0 0.0533 0.8 1>
    R1 Crimson:LinearColor=<rgba 1 0 0 1>
    R1 DefaultColor=<rgba 1 0.0658 0.05 1>
    R1 ForestGreen:LinearColor=<rgba 0 1 0 1>
    R1 Grey:LinearColor=<rgba 1 1 1 1>
    R1 Lime:LinearColor=<rgba 0.2333 0.4 0 1>
    R1 Orange:LinearColor=<rgba 0.6 0.05 0 1>
    R1 Pink:LinearColor=<rgba 0.95 0.0475 0.7996 1>
    R1 Purple:LinearColor=<rgba 0.6333 0 0.95 1>
    R1 Saffron:LinearColor=<rgba 1 0.8333 0 1>
    R1 SkyBlue:LinearColor=<rgba 0 0.8 1 1>
    R1 TitaniumWhite:LinearColor=<rgba 1 1 1 1>
    Saffron=<rgba 1 0.8333 0 1>
    SkyBlue=<rgba 0 0.8 1 1>
    TitaniumWhite=<rgba 1 1 1 1>
    Use R1 Painted Overrides=true
   )
   &InitializeParticle(
    Color=$Particles.Color
    Lifetime=&MaskFloatBySpawnGroup(Masked Float=0.2, Passthrough Float=&RandomRangeFloat(Minimum=0.3, Maximum=0.4), Spawn Group Mask=1)
    Mesh Scale Mode=::NewEnumerator3
    Mesh Uniform Scale=0.34
    Sprite Rotation Mode=::NewEnumerator2
    Sprite Size Mode=::NewEnumerator3
    Uniform Sprite Size=&MergeHelper_FloatSwitch(
     0=&RandomRangeFloat(Minimum=200, Maximum=250)
     1=0
     2=&RandomRangeFloat(Minimum=&RandomRangeFloat(Minimum=90, Maximum=140), Maximum=200)
     3=?
     Index=&SpawnGroup()
     Num=::NewEnumerator1
    )
    UsePositionOffset=false
    Write Material Random=true
   )
   &ShapeLocation(
    MaskSpawnGroup=false
    Offset=&AddVectorToPosition(
     Position=&AddVectorToPosition(Position=$User.SingleBoostPos, Vector=$User.DTCarOffset)
     Vector=<-100 0 0>
    )
    Offset Coordinate Space=::Local
    Offset Mode=::NewEnumerator0
    Override Seed=false
    Sphere Radius=&MergeHelper_FloatSwitch(
     0=20
     1=&RandomRangeFloat(Minimum=0, Maximum=10)
     2=30
     3=?
     Index=&SpawnGroup()
     Num=::NewEnumerator1
    )
    Sphere Surface Distribution=0.15
    UseOwnerScale=true
   )
   &AddVelocity(
    Module.Use Falloff Distance=false
    Override Seed=false
    Velocity Mode=::NewEnumerator1
    Velocity Speed=&MaskFloatBySpawnGroup(Masked Float=&RandomRangeFloat(Minimum=-600, Maximum=600), Passthrough Float=0)
   )
   /SetVariables(
    Particles.R1 Color:LinearColor=&MultiplyLinearColor(
     Linear Color=$Particles.R1 Color
     Scale Factor=&MakeVector4FromVectorAndFloat(
      "Vector (RGB)"=&VectorFromFloat(Value=&Multiply_Float(
        A=&FloatFromCurve(FloatCurve=DataInterface() {propertyValues {Curve {
            defaultValue=1
            keys [
             {interpMode=RCIM_Cubic, tangentMode=RCTM_User, time=0, value=0.5}
             {interpMode=RCIM_Cubic, tangentMode=RCTM_User, time=0.5, value=1}
             {tangentMode=RCTM_User, time=1, value=0.5}
            ]
           }}})
        B=&Multiply_Float(
         A=&SelectFloatFromArray(
          Direct Array Index=$User.ColorVariantID
          Float Selection Array=DataInterface() {propertyValues {FloatData [50,25,50,50,50,5,30,30,50,50,50,20,50,30,50]}}
         )
         B=/VehicleCosmetics/FX/Niagara/Modules/VC_Mask_Floats(Coro Value:NiagaraFloat=1, FrontEnd Value:NiagaraFloat=2, FN Value:NiagaraFloat=0.5)
        )
       ))
      "Float (Alpha)"=1
     )
    )
    Particles.smokecolor:LinearColor=<rgba 0.025 0.025 0.025 1>
   ) {isSetParameters=true}
   /SetVariables(Particles.VisibilityTag=$Emitter.SpawnGroup) {isSetParameters=true, module=SetVariables_#4}
  ]
  particleUpdateScript [
   &ParticleState()
   &Drag(
    Drag=&MaskFloatBySpawnGroup(Masked Float=2, Passthrough Float=0)
    Use Linear Drag=true
    Use Rotational Drag=false
   )
   &SolveForcesAndVelocity(Clamp Velocity=false, Limit Acceleration=false)
   &SubUVAnimation(
    "Sprite Renderer "=DataInterface() {propertyValues {SpriteRenderer=&self:NE_VC_Cone_0.NiagaraSpriteRendererProperties_0}}
    UseEndFrame=false
    UseStartFrame=false
   )
   &ScaleColor(
    Scale RGB=&VectorFromFloat(Value=&Multiply_Float(
      A=&SelectFloatFromArray(
       Direct Array Index=$User.ColorVariantID
       Float Selection Array=DataInterface() {propertyValues {FloatData [100,1,1,100,30,1.5,5,1,20,100,25,6,8,5,20]}}
      )
      B=&VC_Mask_Floats(Coro Value:NiagaraFloat=1, FrontEnd Value:NiagaraFloat=1, FN Value:NiagaraFloat=0.5)
     ))
    ScaleA=false
    ScaleRGB=true
   )
   &DynamicMaterialParameters(
    Index 1 Param 1=&FloatFromCurve(FloatCurve=DataInterface() {propertyValues {Curve {keys [<at 0.4 -1>,<at 1 1>]}}})
    Index 1 Param 2=0.05
    Module.Index1Float0Write=true
    Module.Index1Float1Write=true
    Module.Index1Float2Write=true
    Module.Index1Float3Write=true
    Write Parameter Index 0=::NewEnumerator0
    Write Parameter Index 1=::NewEnumerator1
   )
  ]
  rendererClasses [&NiagaraSpriteRendererProperties, &NiagaraMeshRendererProperties]
  renderers [
   {rendererClass=&NiagaraSpriteRendererProperties}
   {rendererClass=&NiagaraMeshRendererProperties, rendererIndex=1}
   {rendererClass=&NiagaraSpriteRendererProperties, rendererIndex=2}
  ]
 }
 NE_VC_Cone001 {
  emitterUpdateScript [
   &EmitterState(Inactive Response=::NewEnumerator1, Life Cycle Mode=::NewEnumerator1, Loop Duration=0.2, UseLoopDelay=false)
   &SpawnBurst_Instantaneous(Use Spawn Probability=false, UseLoopCountLimit=false)
   &SpawnBurst_Instantaneous(Spawn Group=1, Use Spawn Probability=false, UseLoopCountLimit=false) {module=SpawnBurst_Instantaneous001}
   &SpawnBurst_Instantaneous(Spawn Group=2, Use Spawn Probability=false, UseLoopCountLimit=false) {module=SpawnBurst_Instantaneous002}
   &SpawnRate(SpawnGroup=2, SpawnRate=9, Use Spawn Probability=false) {enabled=false}
  ]
  particleSpawnScript [
   &VC_PaintedVariantsOverride(
    Black=<rgba 0.03 0.03 0.03 1>
    BurntSienna=<rgba 0.6 0.159 0.012 1>
    Cobalt=<rgba 0 0.0267 0.4 1>
    Crimson=<rgba 1 0 0 1>
    DefaultColor=<rgba 1 0.0658 0.05 1>
    ForestGreen=<rgba 0 1 0 1>
    Grey=<rgba 0.6 0.6 0.6 1>
    Lime=<rgba 0.1667 0.4 0 1>
    Orange=<rgba 0.3 0.01 0 1>
    Pink=<rgba 0.9 0 0.75 1>
    Purple=<rgba 0.3333 0 0.5 1>
    Saffron=<rgba 1 0.8333 0 1>
    SkyBlue=<rgba 0 0.8 1 1>
    TitaniumWhite=<rgba 1 1 1 1>
   )
   &InitializeParticle(
    Color=$Particles.Color
    Lifetime=&MergeHelper_FloatSwitch(
     0=&RandomRangeFloat(Minimum=0.15, Maximum=0.25)
     1=&RandomRangeFloat(Minimum=0.15, Maximum=0.25)
     2=0.3
     3=?
     Index=&SpawnGroup()
     Num=::NewEnumerator1
    )
    Mesh Scale Mode=::NewEnumerator3
    Mesh Uniform Scale=1.1
    Sprite Rotation Mode=::NewEnumerator2
    Sprite Size Mode=::NewEnumerator3
    Uniform Sprite Size=215
    UsePositionOffset=false
    Write Material Random=true
   )
   &ShapeLocation(
    MaskSpawnGroup=false
    Offset=&AddVectorToPosition(
     Position=&AddVectorToPosition(Position=$User.SingleBoostPos, Vector=$User.DTCarOffset)
     Vector=<-100 0 0>
    )
    Offset Coordinate Space=::Local
    Offset Mode=::NewEnumerator0
    Override Seed=false
    Sphere Radius=&MaskFloatBySpawnGroup(Masked Float=2, Passthrough Float=&RandomRangeFloat(Minimum=0, Maximum=10), Spawn Group Mask=2)
    Sphere Surface Distribution=0.1
    UseOwnerScale=true
   )
   &AddVelocity(Module.Use Falloff Distance=false, Override Seed=false, Velocity Mode=::NewEnumerator1) {enabled=false}
   &InitialMeshOrientation(Use Rotation Vector=false)
   /SetVariables(
    Particles.glowcolor:LinearColor=&MultiplyLinearColor(
     Linear Color=$Particles.Color
     Scale Factor=&MakeVector4FromVectorAndFloat(
      "Vector (RGB)"=&VectorFromFloat(Value=&Multiply_Float(
        A=&SelectFloatFromArray(
         Direct Array Index=$User.ColorVariantID
         Float Selection Array=DataInterface() {propertyValues {FloatData [40,40,30,30,40,20,10,5,50,50,20,50,30,20,20]}}
        )
        B=&VC_Mask_Floats(Coro Value:NiagaraFloat=1, FrontEnd Value:NiagaraFloat=1, FN Value:NiagaraFloat=0.5)
       ))
      "Float (Alpha)"=0.2
     )
    )
    Particles.insidecolor:LinearColor=<rgba 0.005 0.005 0.005 1>
   ) {isSetParameters=true}
   /SetVariables(Particles.VisibilityTag=$Emitter.SpawnGroup) {isSetParameters=true, module=SetVariables_#4}
  ]
  particleUpdateScript [
   &ParticleState()
   &UpdateMeshOrientation(Rotation Rate=0.25, Rotation Vector=&RandomVector(Vector Scale=1))
   &SolveForcesAndVelocity(Clamp Velocity=false, Limit Acceleration=false)
   &ScaleColor(
    Scale Alpha=2
    Scale RGB=&VectorFromFloat(Value=&Multiply_Float(
      A=&SelectFloatFromArray(
       Direct Array Index=$User.ColorVariantID
       Float Selection Array=DataInterface() {propertyValues {FloatData [250,1,1,200,30,1.5,5,2,20,150,20,5,10,5,10]}}
      )
      B=&VC_Mask_Floats(Coro Value:NiagaraFloat=1, FrontEnd Value:NiagaraFloat=1, FN Value:NiagaraFloat=0.5)
     ))
    ScaleA=true
    ScaleRGB=true
   )
   &DynamicMaterialParameters(
    Index 1 Param 1=&FloatFromCurve(FloatCurve=DataInterface() {propertyValues {Curve {keys [<at 0 0>, {arriveTangent=1.2091, leaveTangent=1.2091, time=0.8, value=1}]}}})
    Index 1 Param 2=0.1
    Module.Index1Float0Write=true
    Module.Index1Float1Write=true
    Module.Index1Float2Write=true
    Module.Index1Float3Write=true
    Write Parameter Index 0=::NewEnumerator0
    Write Parameter Index 1=::NewEnumerator1
   )
   /SetVariables(Particles.fillinside:Vector4f=&MakeVector4(
     X=&FloatFromCurve(FloatCurve=DataInterface() {propertyValues {Curve {keys [{arriveTangent=4.7219, time=0.2, value=-1}, {arriveTangent=-1.1375, time=0.85, value=1}]}}})
     Y=0.2
     Z=0
     W=0
    )) {isSetParameters=true}
  ]
  rendererClasses [&NiagaraSpriteRendererProperties, &NiagaraMeshRendererProperties]
  renderers [
   {rendererClass=&NiagaraSpriteRendererProperties}
   {rendererClass=&NiagaraMeshRendererProperties, rendererIndex=1}
   {rendererClass=&NiagaraMeshRendererProperties, rendererIndex=2}
   {rendererClass=&NiagaraMeshRendererProperties, rendererIndex=3}
   {rendererClass=&NiagaraMeshRendererProperties, rendererIndex=4}
  ]
 }
 NE_VC_Smoke {
  emitterUpdateScript [
   &EmitterState(Inactive Response=::NewEnumerator1, Life Cycle Mode=::NewEnumerator1, Loop Duration=1, UseLoopDelay=false)
   &SpawnRate(SpawnRate=50, Use Spawn Probability=false)
   &SpawnRate(SpawnGroup=1, SpawnRate=65, Use Spawn Probability=false) {module=SpawnRate001}
  ]
  particleSpawnScript [
   &VC_PaintedVariantsOverride(
    Black=<rgba 0 0 0 1>
    BurntSienna=<rgba 0.6 0.33 0.06 1>
    Cobalt=<rgba 0 0.2167 0.65 1>
    Crimson=<rgba 1 0 0 1>
    DefaultColor=<rgba 1 0.0658 0.05 1>
    ForestGreen=<rgba 0 1 0 1>
    Grey=<rgba 1 1 1 1>
    Lime=<rgba 0.3667 0.5 0 1>
    Orange=<rgba 0.6 0.06 0 1>
    Pink=<rgba 0.9 0.018 0.753 1>
    Purple=<rgba 0.4 0 0.6 1>
    Saffron=<rgba 1 0.8333 0 1>
    SkyBlue=<rgba 0 0.8 1 1>
    TitaniumWhite=<rgba 1 1 1 1>
   )
   &InitializeParticle(
    Color=$Particles.Color
    Lifetime=&MaskFloatBySpawnGroup(
     Masked Float=&RandomRangeFloat(Minimum=0.4, Maximum=0.6)
     Passthrough Float=&RandomRangeFloat(Minimum=0.3, Maximum=0.4)
    )
    Mesh Scale Mode=::NewEnumerator1
    Mesh Uniform Scale Max=1.8
    Mesh Uniform Scale Min=1.4
    Position Offset=&AddVectorToPosition(Position=<0 0 0>, Vector=<0 0 0>)
    Sprite Rotation Mode=::NewEnumerator2
    Sprite Size Mode=::NewEnumerator1
    Uniform Sprite Size Max=200
    Uniform Sprite Size Min=150
    UsePositionOffset=false
    Write Material Random=true
   )
   &AddVelocity(Velocity=&MaskVectorBySpawnGroup(
     Masked Vector=&RandomRangeVector(Minimum=<-1500 0 0>, Maximum=<-2000 0 0>)
     Passthrough Vector=&RandomRangeVector(Minimum=<-1500 0 0>, Maximum=<-2000 0 0>)
    ))
   &ShapeLocation(
    MaskSpawnGroup=false
    Offset=&AddVectorToPosition(
     Position=&AddVectorToPosition(Position=$User.SingleBoostPos, Vector=$User.DTCarOffset)
     Vector=<-150 0 0>
    )
    Offset Coordinate Space=::Local
    Offset Mode=::NewEnumerator0
    Override Seed=false
    Sphere Radius=&MaskFloatBySpawnGroup(Masked Float=20, Passthrough Float=30)
    Sphere Surface Distribution=0.2
    UseOwnerScale=true
   )
   &InitialMeshOrientation(Use Rotation Vector=false)
   /SetVariables(Particles.smokecolor:LinearColor=<rgba 0.025 0.025 0.025 1>) {isSetParameters=true}
   /SetVariables(Particles.VisibilityTag=$Emitter.SpawnGroup) {isSetParameters=true}
  ]
  particleUpdateScript [
   &ParticleState()
   &ScaleMeshSize(Scale Factor=&VectorFromFloat(Value=&FloatFromCurve(FloatCurve=DataInterface() {propertyValues {Curve {keys [
          {arriveTangent=0.679, interpMode=RCIM_Cubic, leaveTangent=0.679, tangentMode=RCTM_User, time=0, value=1}
          {arriveTangent=-0.2191, interpMode=RCIM_Cubic, leaveTangent=-0.2191, tangentMode=RCTM_User, time=1, value=0.2}
         ]}}})))
   &ScaleSpriteSize(Uniform Curve Sprite Scale=DataInterface() {propertyValues {Curve {keys [{tangentMode=RCTM_User, time=0, value=1}, {tangentMode=RCTM_User, time=1, value=3}]}}})
   &Drag(
    Drag=&FloatFromCurve(FloatCurve=DataInterface() {propertyValues {Curve {
        defaultValue=1
        keys [
         {arriveTangent=0.3705, interpMode=RCIM_Cubic, leaveTangent=0.3705, tangentMode=RCTM_User, time=0, value=0}
         {arriveTangent=0.7457, interpMode=RCIM_Cubic, leaveTangent=0.7457, tangentMode=RCTM_User, time=1, value=1}
        ]
       }}})
    Use Linear Drag=true
    Use Rotational Drag=false
   )
   &SolveForcesAndVelocity(Clamp Velocity=false, Limit Acceleration=false)
   &SubUVAnimation(
    "Sprite Renderer "=DataInterface() {propertyValues {SpriteRenderer=&self:NE_VC_Smoke_0.NiagaraSpriteRendererProperties_0}}
    UseEndFrame=false
    UseStartFrame=false
   )
   &ScaleColor(
    Scale RGB=&VectorFromFloat(Value=&Multiply_Float(
      A=&SelectFloatFromArray(
       Direct Array Index=$User.ColorVariantID
       Float Selection Array=DataInterface() {propertyValues {FloatData [20,1,1,20,25,10,10,2,20,20,10,20,10,10,10]}}
      )
      B=&VC_Mask_Floats(Coro Value:NiagaraFloat=1, FrontEnd Value:NiagaraFloat=1, FN Value:NiagaraFloat=0.5)
     ))
    ScaleA=true
    ScaleRGB=true
   )
   &DynamicMaterialParameters(
    Index 1 Param 1=&FloatFromCurve(FloatCurve=DataInterface() {propertyValues {Curve {keys [<at 0.4 -1>,<at 1 1>]}}})
    Index 1 Param 2=0.05
    Module.Index1Float0Write=true
    Module.Index1Float1Write=true
    Module.Index1Float2Write=true
    Module.Index1Float3Write=true
    Write Parameter Index 0=::NewEnumerator0
    Write Parameter Index 1=::NewEnumerator1
   )
  ]
  rendererClasses [&NiagaraSpriteRendererProperties, &NiagaraMeshRendererProperties]
  renderers [
   {rendererClass=&NiagaraSpriteRendererProperties}
   {rendererClass=&NiagaraMeshRendererProperties, rendererIndex=1}
  ]
 }
}
summary {
 systemName=NS_Sphenergy
 emitters [
  {emitterName=NE_VC_Cone, rendererClasses [&NiagaraSpriteRendererProperties, &NiagaraMeshRendererProperties]}
  {emitterName=NE_VC_Smoke, rendererClasses [&NiagaraSpriteRendererProperties, &NiagaraMeshRendererProperties]}
  {emitterName=NE_VC_Cone001, rendererClasses [&NiagaraSpriteRendererProperties, &NiagaraMeshRendererProperties]}
 ]
 userVariables [
  User.bIsFrontendPreview {type=NiagaraBool}
  User.bIsPainted {type=NiagaraBool}
  User.bPreviewVehicleSpeed {type=NiagaraBool}
  User.ColorVariantID {type=NiagaraInt32}
  User.DTCarOffset {type=Vector3f, description="", defaultValue {struct=Vector3f, value=<0 0 0>}}
  User.GameModeID {type=NiagaraInt32}
  User.PaintedColor {type=LinearColor, description="", defaultValue {struct=LinearColor, value=<rgba 1 1 1 1>}}
  User.PaintedScalar {type=NiagaraFloat, defaultValue=1, description=""}
  User.PortCount {type=NiagaraInt32, defaultValue=1}
  User.SingleBoostPos {type=NiagaraPosition, description="", defaultValue {}}
 ]
}
