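# M_DustClouds is an unlit translucent Niagara sprite material for smoky dust clouds in the Sphenergy vehicle boost effect.
# A smoke normal flipbook's red and green channels form fake directional shading, tinted by particle color and scaled by dynamic Brightness into an emissive-adjustment function.
# Flipbook alpha, shaped by dynamic Dissolve (power) and Sharpness (multiplier), is clamped, multiplied by particle alpha and passed through a depth-fade function.
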Material /VehicleCosmetics/Boosts/Boost_Sphenergy/Materials/M_DustClouds
graph {
 properties {bUsedWithNiagaraSprites=true, blendMode=BLEND_Translucent, pixelDepthOffsetMode=PDOM_Legacy, shadingModel=MSM_Unlit, usedShadingModels=Unlit}
 nodes [
  n1 = TextureSampleParameter2D() {parameterName=SubUV, texture=/DealershipSharedLibrary/FX/Textures/Flipbook/T_VCVFX_SmokeHQ_N}
  n2 = ConstantBiasScale(None=DotProduct(A=@n1.RGB, B=Constant3Vector() {constant=<rgba 1 1 0 0>})) {bias=0.2}
  n3 = ParticleColor()
  n4 = DynamicParameter() {paramNames [Dissolve,Sharpness,Brightness,Param4]}
  MaterialFunctionCall(Emissive=Multiply(A=Multiply(A=@n2, B=@n3.RGB), B=@n4.Brightness)) {materialFunction=&MF_FX_EmissiveAdjustments}
  n5 = Multiply(A=Power(Base=@n1.A, Exp=@n4.Dissolve), B=@n4.Sharpness)
  MaterialFunctionCall(Opacity In=Multiply(A=@n3.A, B=Saturate(None=@n5)), DepthFade Distance=Constant() {r=64}) {materialFunction=&MF_VCVFX_DepthFade}
 ]
}
