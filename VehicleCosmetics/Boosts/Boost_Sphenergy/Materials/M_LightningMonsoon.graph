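# Translucent unlit Niagara sprite material for lightning, part of the Sphenergy vehicle boost effect.
# A lightning mask, wobbled by rotating noise, is multiplied by a panning grid mask and noise-distorted water texture for moving breakup.
# Dynamic parameters blend grid pan speed, toggle a fixed quarter-turn, blend between normal and brighter V-flipped lightning, and scale brightness.
# Particle color tints the emissive, routed through MF_FX_EmissiveAdjustments, and particle alpha scales opacity.
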Material /VehicleCosmetics/Boosts/Boost_Sphenergy/Materials/M_LightningMonsoon
graph {
 properties {bUsedWithNiagaraSprites=true, blendMode=BLEND_Translucent, pixelDepthOffsetMode=PDOM_Legacy, shadingModel=MSM_Unlit, usedShadingModels=Unlit}
 nodes [
  n1 = ParticleColor()
  n2 = TextureCoordinate() {uTiling=0.25, vTiling=0.25}
  n3 = DynamicParameter() {defaultValue=<rgba 1 1 0 1>, paramNames [DetailMotion, Brightness, RotateTexture, TextureSwitch]}
  n4 = LinearInterpolate(A=Panner(Coordinate=@n2) {speedX=0.25, speedY=-0.25}, B=Panner(Coordinate=@n2) {speedX=0.25, speedY=-0.15}, Alpha=@n3.DetailMotion)
  n5 = Time()
  n6 = Constant() {r=0.5}
  n7 = TextureSample(UVs=Rotator(Coordinate=TextureCoordinate() {uTiling=0.25, vTiling=0.25}, Time=Multiply(A=@n5, B=@n6)) {speed=-0.2}) {texture=/DealershipSharedLibrary/FX/Textures/Noise/T_VCVFX_NoiseFire01_N}
  n8 = Multiply(A=@n7.RGB) {constB=0.05}
  n9 = Panner(Coordinate=Add(A=TextureCoordinate() {uTiling=2.5, vTiling=2.5}, B=ComponentMask(None=@n8) {g=true, r=true})) {speedX=0.1, speedY=0.5}
  n10 = Add(A=Multiply(A=TextureSample(UVs=@n9) {texture=/DealershipSharedLibrary/FX/Textures/Noise/T_VCVFX_Water01_A}.RGB) {constB=5.5}) {constB=0.1}
  n11 = Multiply(A=@n7.RGB) {constB=0.025}
  n12 = Add(A=ComponentMask(None=@n11) {g=true, r=true}, B=TextureCoordinate())
  n13 = Constant() {r=0.5}
  n14 = TextureSample(UVs=LinearInterpolate(A=Rotator(Coordinate=@n12, Time=@n13) {speed=3.14}, B=@n12, Alpha=@n3.RotateTexture)) {texture=/DealershipSharedLibrary/FX/Textures/Gradient_Masks/T_VCVFX_LightningFlip_A}
  n15 = Add(A=ComponentMask(None=@n11) {g=true, r=true}, B=TextureCoordinate() {vTiling=-1})
  n16 = Constant() {r=0.5}
  n17 = TextureSample(UVs=LinearInterpolate(A=Rotator(Coordinate=@n15, Time=@n16) {speed=3.14}, B=@n15, Alpha=@n3.RotateTexture)) {texture=&T_VCVFX_LightningFlip_A}
  n18 = Multiply(
   A=Multiply(
    A=TextureSample(UVs=@n4) {texture=/DealershipSharedLibrary/FX/Textures/Gradient_Masks/T_VCVFX_WidgetGridVertexColorMaterial_A}.RGB
    B=Saturate(None=@n10)
   )
   B=LinearInterpolate(A=Multiply(A=@n14.RGB) {constB=0.5}, B=Multiply(A=@n17.RGB) {constB=10}, Alpha=@n3.TextureSwitch)
  )
  n19 = Saturate(None=@n18)
  MaterialFunctionCall(Emissive=Multiply(A=Multiply(A=@n1.RGB, B=@n19), B=@n3.Brightness)) {materialFunction=&MF_FX_EmissiveAdjustments}
  Saturate(None=Multiply(A=@n1.A, B=@n18))
 ]
}
