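# Unlit masked material for the ring tail of the Sphenergy vehicle boost effect, usable on Niagara ribbons and mesh particles.
# Panning noise distorts the UVs of scrolling smoke, swirly fluid and energy trail textures, which are summed and shaped by a procedural UV edge fade.
# The result is multiplied by particle color and alpha, with emissive passed through MF_FX_EmissiveAdjustments.
# Dynamic parameters drive brightness, tiling and pan speed; WPO_Noise is unused.
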Material /VehicleCosmetics/Boosts/Boost_Sphenergy/Materials/M_RingTail
graph {
 properties {
  bUsedWithNiagaraMeshParticles=true
  bUsedWithNiagaraRibbons=true
  blendMode=BLEND_Masked
  pixelDepthOffsetMode=PDOM_Legacy
  shadingModel=MSM_Unlit
  usedShadingModels=Unlit
 }
 nodes [
  n1 = DynamicParameter() {paramNames [Brightness,WPO_Noise,Tile,PanSpeed]}
  n2 = MaterialFunctionCall(Float2=TextureCoordinate()) {materialFunction=&BreakOutFloat2Components}
  n3 = Multiply(A=Time(), B=@n1.PanSpeed)
  n4 = TextureSample(UVs=Panner(Coordinate=TextureCoordinate() {vTiling=0.5}, Time=@n3) {speedX=-0.03, speedY=0.07}) {texture=/VehicleCosmetics/Boosts/Boost_Sphenergy/Textures/T_VCVFX_Boost_Sphenergy_T_Noise01}
  n5 = Add(
   A=AppendVector(A=Multiply(A=@n2.R, B=@n1.Tile), B=Multiply(A=@n2.G, B=@n1.Tile))
   B=Multiply(A=@n4.R, B=Constant2Vector() {g=-0.05, r=-0.025})
  )
  n6 = Multiply(A=@n4.R, B=Constant2Vector() {g=0.05, r=0.025})
  n7 = TextureCoordinate() {uTiling=0.3, vTiling=0.1}
  n8 = Panner(Coordinate=Add(A=@n6, B=@n7), Time=@n3) {speedX=-0.1, speedY=0.12}
  n9 = Panner(Coordinate=Add(A=@n6, B=@n7), Time=@n3) {speedX=-0.2}
  n10 = Add(
   A=TextureSample(UVs=Panner(Coordinate=@n5, Time=@n3) {speedX=-0.2}) {texture=/VehicleCosmetics/Boosts/Boost_Sphenergy/Textures/T_VCVFX_Boost_Sphenergy_SimpleLitSmoke_D}.R
   B=Add(
    A=TextureSample(UVs=@n8) {texture=/VehicleCosmetics/Boosts/Boost_Sphenergy/Textures/T_VCVFX_Boost_Sphenergy_swirlyFluid}.R
    B=TextureSample(UVs=@n9) {texture=/VehicleCosmetics/Boosts/Boost_Sphenergy/Textures/T_VCVFX_Boost_Sphenergy_energyTrail}.R
   )
  )
  n11 = MaterialFunctionCall(Float2=TextureCoordinate()) {materialFunction=&BreakOutFloat2Components}
  n12 = OneMinus(None=@n11.G)
  n13 = Saturate(None=Multiply(A=Multiply(A=@n11.G, B=@n12), B=Constant() {r=3}))
  n14 = Saturate(None=OneMinus(None=@n11.R))
  n15 = Multiply(A=@n14, B=@n11.R)
  n16 = Multiply(A=Multiply(A=Multiply(A=@n13, B=@n14), B=Multiply(A=@n15) {constB=10}), B=Constant() {r=5})
  n17 = Saturate(None=ConstantBiasScale(None=Multiply(A=@n10, B=@n16)) {bias=-0.5, scale=0.2})
  n18 = ParticleColor()
  MaterialFunctionCall(Emissive=Multiply(A=Multiply(A=@n1.Brightness, B=@n17), B=@n18.RGB)) {materialFunction=&MF_FX_EmissiveAdjustments}
  Multiply(A=@n17, B=@n18.A)
 ]
}
