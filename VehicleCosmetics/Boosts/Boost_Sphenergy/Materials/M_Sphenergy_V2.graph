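# Unlit masked material for Niagara mesh particles in the Sphenergy vehicle boost, drawing stylized fire and smoke.
# Panning cell noise distorts a panning boost noise, which combines with fresnel to erode particle alpha.
# That mask feeds a derived normal for hard-edged fake smoke lighting and breaks up a view-based core quantized by FireSteps.
# Dynamic parameters drive core, darkness, tiling and pan speed, and particle color tints the emissive.
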Material /VehicleCosmetics/Boosts/Boost_Sphenergy/Materials/M_Sphenergy_V2
graph {
 properties {
  bTangentSpaceNormal=false
  bUsedWithNiagaraMeshParticles=true
  blendMode=BLEND_Masked
  opacityMaskClipValue=0.001
  pixelDepthOffsetMode=PDOM_Legacy
  shadingModel=MSM_Unlit
  usedShadingModels=Unlit
 }
 nodes [
  n1 = ParticleColor()
  n2 = DynamicParameter() {paramNames [Core, SmokeDarkness, tile, PanSpeed]}
  n3 = Multiply(A=@n2.tile, B=TextureCoordinate() {uTiling=2})
  n4 = MaterialFunctionCall(Speed=@n2.PanSpeed) {materialFunction=&TimeWithSpeedVariable}
  n5 = StaticSwitchParameter(True=Constant2Vector() {r=0.25}, False=Constant2Vector() {g=0.25}) {parameterName=invertVelocity}
  n6 = Multiply(A=TextureCoordinate(), B=ScalarParameter() {defaultValue=1, parameterName=tile})
  n7 = MaterialFunctionCall(Speed=ScalarParameter() {defaultValue=1, parameterName=Speed}) {materialFunction=&TimeWithSpeedVariable}
  n8 = ComponentMask(None=TextureSample(UVs=Panner(Coordinate=@n6, Time=@n7.Time, Speed=@n5) {speedY=0.25}) {texture=/DealershipSharedLibrary/FX/Textures/Noise/T_VCVFX_CellNoise01_A}.RGB) {g=true, r=true}
  n9 = ScalarParameter() {defaultValue=1, parameterName=Intensity}
  n10 = TextureSample(UVs=Add(A=Panner(Coordinate=@n3, Time=@n4.Time, Speed=@n5) {speedY=0.25}, B=Multiply(A=@n8, B=@n9))) {texture=/VehicleCosmetics/Boosts/Boost_Sphenergy/Textures/T_VCVFX_Boost_Sphenergy_T_Noise01}
  n11 = Fresnel()
  n12 = Divide(A=Add(A=OneMinus(None=@n10.RGB), B=@n11)) {constB=3}
  n13 = Saturate(None=Subtract(A=@n1.A, B=Saturate(None=@n12)))
  ComponentMask(None=@n13) {r=true}
  n14 = Constant3Vector() {constant=<rgba 0 0 1 0>}
  n15 = DeriveNormalZ(InXY=ComponentMask(None=@n13) {g=true, r=true})
  n16 = DotProduct(A=Transform(None=Normalize(VectorInput=@n14)), B=@n15)
  n17 = Ceil(None=Power(Base=Subtract(A=@n16) {constB=0.5}, Exp=ScalarParameter() {defaultValue=8, parameterName=EXP_Tex1}) {constExponent=5})
  n18 = Saturate(None=Power(Base=Divide(A=@n17), Exp=ScalarParameter() {defaultValue=1, parameterName=Exp_Tex3}) {constExponent=1})
  n19 = Multiply(A=VectorParameter() {parameterName=SmokeColor, defaultValue=<rgba 1 1 1 1>}.A, B=Saturate(None=Add(A=@n18) {constB=0.5}))
  n20 = CameraVectorWS()
  n21 = Power(Base=DotProduct(A=ReflectionVectorWS(), B=@n20), Exp=ScalarParameter() {defaultValue=1, parameterName=CenterFalloff}) {constExponent=8}
  n22 = ScalarParameter() {defaultValue=4, parameterName=FireSteps}
  n23 = Multiply(A=Add(A=@n21, B=@n2.Core), B=@n22)
  n24 = OneMinus(None=@n13)
  n25 = Divide(A=Ceil(None=Subtract(A=@n23, B=@n24)), B=@n22)
  n26 = Saturate(None=Power(Base=Subtract(A=@n25) {constB=0.5}, Exp=ScalarParameter() {defaultValue=1, parameterName=Exp_Tex4}) {constExponent=5})
  n27 = Power(
   Base=LinearInterpolate(A=Multiply(A=@n19), B=Multiply(A=@n26, B=@n2.SmokeDarkness), Alpha=@n26)
   Exp=ScalarParameter() {defaultValue=1, parameterName=ssPowEmmisive}
  )
  MaterialFunctionCall(Emissive=Multiply(A=Saturate(None=@n27), B=@n1.RGB)) {materialFunction=&MF_FX_EmissiveAdjustments}
  n28 = MaterialFunctionCall(Float3=@n20) {materialFunction=&BreakOutFloat3Components}
  MaterialFunctionCall(X=Multiply(A=@n28.G) {constB=-1}, Y=Multiply(A=@n28.B) {constB=-1}, Z=@n28.R) {materialFunction=&MakeFloat3}
  OneMinus(None=@n11)
  Power(Base=DotProduct(A=Saturate(None=@n21), B=@n15), Exp=ScalarParameter() {defaultValue=2, parameterName=EXP_Tex2})
  Saturate(None=Power(Base=@n13, Exp=ScalarParameter() {defaultValue=1, parameterName=PowOpacity}))
 ]
}
